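# Niagara system for the AuroraOwl emote's ground effects: smoke rings, warm glows, flickering lights and a despawn burst.
# System-age curves time the hand-offset rings first, then feet rings, with Glow_FE following similar timing and bIsFrontEnd flags suppressing one-shot ring bursts.
# Scratch trace modules orient rings to the ground normal, skipping traces by platform quality.
# DespawnBurst fires after a delay, using spawn groups for distinct velocity, gravity and lifetime.
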NiagaraSystem /BRCosmetics/Animation/Game/MainPlayer/Emotes/AuroraOwl/FX/NS_Emote_AuroraOwl_GroundStuff
dependencies {
 usedDataInterfaces [
  &NiagaraDataInterfaceCurve
  &NiagaraDataInterfaceCollisionQuery
  &NiagaraDataInterfaceDebugDraw
  &NiagaraDataInterfacePlatformSet
  &NiagaraDataInterfaceColorCurve
  &NiagaraDataInterfaceSkeletalMesh
 ]
 usedDynamicInputs [
  &SimulationPosition
  &Multiply_Vector2D_ByFloat
  &MakeVector4_FromLinearColor
  &FNTimeOfDayColor
 ]
 usedModules [
  &SystemState
  &self:SystemSpawnScript.SystemScriptSource.SystemScriptGraph.NiagaraNodeAssignment_0.SetVariables_#1
  &self:SystemSpawnScript.SystemScriptSource.SystemScriptGraph.NiagaraNodeAssignment_1.SetVariables_#2
 ]
}
emitters {
 DespawnBurst {
  emitterUpdateScript [
   &EmitterState(Life Cycle Mode=::NewEnumerator1, Loop Behavior=::NewEnumerator1, Loop Delay=4.15, Loop Duration=1.1363, UseLoopDelay=true)
   &SpawnBurst_Instantaneous(Spawn Count=10, Spawn Group=2, Use Spawn Probability=false, UseLoopCountLimit=false) {module=SpawnBurst_Instantaneous001}
   &SpawnBurst_Instantaneous(Spawn Count=10, Spawn Group=2, Spawn Time=0.3, Use Spawn Probability=false, UseLoopCountLimit=false) {module=SpawnBurst_Instantaneous006}
   &SpawnBurst_Instantaneous(Spawn Count=10, Spawn Group=1, Use Spawn Probability=false, UseLoopCountLimit=false) {module=SpawnBurst_Instantaneous005}
   &SpawnBurst_Instantaneous(Spawn Count=10, Spawn Group=1, Spawn Time=0.25, Use Spawn Probability=false, UseLoopCountLimit=false) {module=SpawnBurst_Instantaneous002}
   &SpawnBurst_Instantaneous(Spawn Count=10, Use Spawn Probability=false, UseLoopCountLimit=false)
   &SpawnBurst_Instantaneous(Spawn Count=5, Spawn Time=0.3, Use Spawn Probability=false, UseLoopCountLimit=false) {module=SpawnBurst_Instantaneous003}
   &SpawnBurst_Instantaneous(Spawn Count=5, Spawn Time=0.1363, Use Spawn Probability=false, UseLoopCountLimit=false) {module=SpawnBurst_Instantaneous004}
  ]
  particleSpawnScript [
   &PartitionParticles(Number Of Partitions=7, Partition Mode:ENiagara_PartitionMode_v2=::NewEnumerator1)
   &InitializeParticle(
    Color Maximum=<rgba 0.7292 0.6754 0.5536 1>
    Color Minimum=<rgba 0.8594 0.4846 0.3446 1>
    Color Mode=::NewEnumerator2
    Lifetime=&MergeHelper_FloatSwitch(
     0=&RandomRangeFloat(Minimum=0.3, Maximum=0.6)
     1=&RandomRangeFloat(Minimum=0.4, Maximum=0.6)
     2=&RandomRangeFloat(Minimum=0.25, Maximum=0.4)
     3=?
     Index=&SpawnGroup()
     Num=::NewEnumerator1
    )
    Position Offset=<-80 11 30>
    Sprite Rotation Mode=::NewEnumerator2
    Sprite Size Mode=::NewEnumerator1
    Uniform Sprite Size Max=30
    Uniform Sprite Size Min=20
    UsePositionOffset=true
    Write Material Random=true
   )
   /SetVariables(
    Particles.Debris Color:LinearColor=&LerpLinearColors(StartColor=<rgba 0.1615 0.0184 0 1>, EndColor=<rgba 0.2448 0.0279 0 1>, LerpFactor=1)
    Particles.Debris Sprite Size:Vector2f=&Vector2DFromFloat(Value=&RandomRangeFloat(Minimum=3, Maximum=5))
   ) {isSetParameters=true}
   &ShapeLocation(
    Cylinder Height=20
    Cylinder Radius=90
    MaskSpawnGroup=false
    Non Uniform Scale=<1 0.5 1>
    Offset Mode=::NewEnumerator0
    Override Seed=false
    Rotation Mode=::NewEnumerator2
    Shape Primitive=::NewEnumerator1
    Surface Only=false
    UseOwnerScale=true
    Yaw / Pitch / Roll=<0.25 0 0>
   )
   &AddVelocity(Velocity=&MergeHelper_VectorSwitch(
     0=&RandomRangeVector(Minimum=<-10 -10 2>, Maximum=<10 10 10>)
     1=&RandomRangeVector(Minimum=<-200 -200 150>, Maximum=<200 200 300>)
     2=&RandomRangeVector(Minimum=<-200 -200 200>, Maximum=<200 200 400>)
     3=?
     Index=&SpawnGroup()
     Num=::NewEnumerator1
    ))
   /SetVariables(Particles.VisibilityTag=&SpawnGroup()) {isSetParameters=true}
   &CameraOffset(Camera Offset Amount=&MergeHelper_FloatSwitch(0=15, 1=0, 2=0, 3=?, Index=&SpawnGroup(), Num=::NewEnumerator1))
  ]
  particleUpdateScript [
   &ParticleState()
   &DynamicMaterialParameters(
    Index 0 Param 1=&FloatFromCurve(FloatCurve=DataInterface() {propertyValues {Curve {keys [<at 0 0 RCIM_Cubic>,<at 1.0011 0.2>]}}})
    Index 0 Param 2=&FloatFromCurve(FloatCurve=DataInterface() {propertyValues {Curve {keys [
         {arriveTangent=0.0453, interpMode=RCIM_Cubic, tangentMode=RCTM_Break, time=0, value=-1}
         {arriveTangent=0.0604, interpMode=RCIM_Cubic, leaveTangent=0.0604, tangentMode=RCTM_User, time=1, value=-1}
        ]}}})
    Index 0 Param 3=&FloatFromCurve(FloatCurve=DataInterface() {propertyValues {Curve {keys [{interpMode=RCIM_Cubic, leaveTangent=4.6779, time=0.2851, value=0}]}}})
    Index 1 Param 1=&FloatFromCurve(
     FloatCurve=DataInterface() {propertyValues {Curve {keys [<at 0 0>,<at 0.25 1>]}}}
     Scale Curve=&RandomRangeFloat(Minimum=2, Maximum=1)
    )
    Index 1 Param 2=&FloatFromCurve(
     FloatCurve=DataInterface() {propertyValues {Curve {keys [
         {interpMode=RCIM_Cubic, leaveTangent=3.7422, tangentMode=RCTM_Break, time=0, value=0.2}
         {interpMode=RCIM_Cubic, leaveTangent=0.9066, time=1, value=2}
        ]}}}
     Scale Curve=&RandomRangeFloat(Minimum=0.5, Maximum=1)
    )
    Index 1 Param 3=&FloatFromCurve(FloatCurve=DataInterface() {propertyValues {Curve {keys [
         {interpMode=RCIM_Cubic, leaveTangent=2.8445, tangentMode=RCTM_Break, time=0, value=0.2}
         {interpMode=RCIM_Cubic, leaveTangent=63.2978, time=0.25, value=1}
        ]}}})
    Index 1 Param 4=&FloatFromCurve(
     FloatCurve=DataInterface() {propertyValues {Curve {keys [<at 0 1>,<at 1 0.8>]}}}
     Scale Curve=&RandomRangeFloat(Minimum=0.5, Maximum=1)
    )
    Index 2 Param 1=&FloatFromCurve(FloatCurve=DataInterface() {propertyValues {Curve {keys [{interpMode=RCIM_Cubic, leaveTangent=3.2309, tangentMode=RCTM_Break, time=0.1, value=0}, <at 1 1>]}}})
    Index 2 Param 2=$Particles.MaterialRandom
    Index 2 Param 3=&FloatFromCurve(FloatCurve=DataInterface() {propertyValues {Curve {keys [
         {interpMode=RCIM_Cubic, tangentMode=RCTM_Break, time=0, value=1}
         {interpMode=RCIM_Cubic, leaveTangent=11.1233, tangentMode=RCTM_Break, time=1, value=2}
        ]}}})
    Index 2 Param 4=0.75
    Module.Index0Float0Write=true
    Module.Index0Float1Write=true
    Module.Index0Float2Write=true
    Module.Index0Float3Write=true
    Module.Index1Float0Write=true
    Module.Index1Float1Write=true
    Module.Index1Float2Write=true
    Module.Index1Float3Write=true
    Module.Index2Float0Write=true
    Module.Index2Float1Write=true
    Module.Index2Float2Write=true
    Module.Index2Float3Write=true
    Write Parameter Index 1=::NewEnumerator1
    Write Parameter Index 2=::NewEnumerator1
   )
   /SetVariables(
    Particles.Sparks Color:LinearColor=<rgba 20 3.4136 0.0252 1>
    Particles.Sparks Size:Vector2f=&RandomRangeVector2D(Minimum=<1 8>, Maximum=<3 14>)
   ) {isSetParameters=true}
   /SetVariables(
    Particles.Debris Color:LinearColor=&MultiplyLinearColor(
     Linear Color=$Particles.Initial.Debris Color
     Scale Factor=&MakeVector4FromVectorAndFloat(
      "Vector (RGB)"=&VectorFromFloat(Value=&FloatFromCurve(FloatCurve=DataInterface() {propertyValues {Curve {keys [<at 0.1 1>,<at 0.4 0.1>,<at 1 0>]}}}, Scale Curve=2000))
      "Float (Alpha)"=&FloatFromCurve(FloatCurve=DataInterface() {propertyValues {Curve {keys [<at 0 0>,<at 0.3027 1>,<at 1 1>]}}})
     )
    )
    Particles.Debris Sprite Size:Vector2f=&Multiply_Vector2D(
     A=$Particles.Initial.Debris Sprite Size
     B=&Vector2DFromFloat(Value=&FloatFromCurve(FloatCurve=DataInterface() {propertyValues {Curve {keys [
           {interpMode=RCIM_Cubic, leaveTangentWeight=0.2874, tangentMode=RCTM_User, time=0, value=1}
           {arriveTangentWeight=0.4064, interpMode=RCIM_Cubic, leaveTangentWeight=0.065, tangentMode=RCTM_User, time=0.8621, value=1}
           <at 1 0 RCIM_Cubic>
          ]}}}))
    )
   ) {isSetParameters=true}
   &ScaleColor(
    Scale Alpha=&FloatFromCurve(FloatCurve=DataInterface() {propertyValues {Curve {keys [<at 0 0>,<at 0.2049 0.9765>,<at 0.7791 0.9107>,<at 1 0>]}}})
    ScaleA=true
    ScaleRGB=true
   )
   &ScaleSpriteSize(Scale Factor=&Vector2DFromFloat(Value=&FloatFromCurve(FloatCurve=DataInterface() {propertyValues {Curve {keys [
          {arriveTangent=5.2354, interpMode=RCIM_Cubic, leaveTangent=5.2354, tangentMode=RCTM_User, time=-0.0046, value=0.25}
          <at 0.1439 1.4923>
          <at 1.0011 3.4809 RCIM_Cubic>
         ]}}})))
   &GravityForce(Gravity=&MergeHelper_VectorSwitch(0=<0 0 0>, 1=<0 0 -500>, 2=<0 0 -980>, 3=?, Index=&SpawnGroup(), Num=::NewEnumerator1))
   &SolveForcesAndVelocity(Clamp Velocity=false, Limit Acceleration=false)
  ]
  rendererClasses [&NiagaraSpriteRendererProperties]
  renderers [
   {rendererClass=&NiagaraSpriteRendererProperties}
   {rendererClass=&NiagaraSpriteRendererProperties, rendererIndex=1}
   {rendererClass=&NiagaraSpriteRendererProperties, rendererIndex=2}
  ]
 }
 Glow_FE {
  emitterUpdateScript [
   &EmitterState()
   &SpawnRate(
    SpawnRate=&FloatFromCurve(
     FloatCurve=DataInterface() {propertyValues {Curve {keys [<at 0.53 4>, {arriveTangent=-2.2759, interpMode=RCIM_Cubic, leaveTangent=-2.2759, time=2.2789, value=3}, <at 2.2876 0 RCIM_Constant>]}}}
     CurveIndex=$System.Age
     Scale Curve=1.3
    )
    Use Spawn Probability=false
   )
   &SpawnRate(
    SpawnGroup=1
    SpawnRate=&FloatFromCurve(
     FloatCurve=DataInterface() {propertyValues {Curve {keys [
         <at 0 0 RCIM_Constant>
         {
          arriveTangent=0.6761
          arriveTangentWeight=0.0188
          interpMode=RCIM_Cubic
          leaveTangent=0.6761
          leaveTangentWeight=1.3602
          time=2.1627
          value=4
         }
         {arriveTangent=-1.3671, interpMode=RCIM_Constant, tangentMode=RCTM_User, time=4.4374, value=3}
         <at 4.4521 0 RCIM_Constant>
        ]}}}
     CurveIndex=$System.Age
     Scale Curve=1.3
    )
    Use Spawn Probability=false
   ) {module=SpawnRate001}
   /SetVariables(Emitter.FE:NiagaraBool=&Boolean_OrOperation(Bool1=$User.bIsFrontEnd, Bool2=$User.bIsFrontEndPreview)) {isSetParameters=true}
  ]
  particleSpawnScript [
   &InitializeParticle(
    Color=<rgba 0.9427 0.1733 0.0468 0.3>
    Mass Mode=::NewEnumerator1
    Position Offset=&MergeHelper_VectorSwitch(0=<65 31 1>, 1=<0 0 1>, 2=?, 3=?, Index=&SpawnGroup(), Num=::NewEnumerator0)
    Sprite Size Mode=::NewEnumerator3
    Uniform Sprite Size=&MergeHelper_FloatSwitch(0=200, 1=500, 2=?, 3=?, Index=&SpawnGroup(), Num=::NewEnumerator0)
    UsePositionOffset=true
    Write Material Random=true
   ) {module=InitializeParticle001}
   &SpriteFacingAndAlignment(Sprite Facing=<0 0 1>, WriteAlignment=false, WriteFacing=true)
   &FN_SetSpawnGroup(Number of Spawn Groups=2)
  ]
  particleUpdateScript [
   &ParticleState()
   &ScaleSpriteSize(
    Uniform Curve Scale=0.9
    Uniform Curve Sprite Scale=DataInterface() {propertyValues {Curve {keys [{tangentMode=RCTM_User, time=0, value=0.5118}, {tangentMode=RCTM_User, time=1, value=1}]}}}
   )
   &ScaleSpriteSize(
    Scale Sprite Size Mode=::NewEnumerator0
    Uniform Scale Factor=&MergeHelper_FloatSwitch(
     0=&FloatFromCurve(
      FloatCurve=DataInterface() {propertyValues {Curve {keys [{tangentMode=RCTM_User, time=0.53, value=1}, {tangentMode=RCTM_User, time=2.2, value=0.5}]}}}
      CurveIndex=$System.Age
     )
     1=&Multiply_Float(A=0, B=$System.Age)
     2=?
     3=?
     Index=&SpawnGroup()
     Num=::NewEnumerator0
    )
   ) {module=ScaleSpriteSize001}
   &ScaleColor(
    Scale Alpha=&FloatFromCurve(FloatCurve=DataInterface() {propertyValues {Curve {keys [<at 0 0 RCIM_Cubic>,<at 0.4237 0.934 RCIM_Cubic>,<at 1 0 RCIM_Cubic>]}}})
    ScaleA=true
    ScaleRGB=true
   )
  ]
  rendererClasses [&NiagaraSpriteRendererProperties]
  renderers [{rendererClass=&NiagaraSpriteRendererProperties}]
 }
 SmokeGroundRings_Feet {
  emitterUpdateScript [
   &EmitterState()
   &SpawnRate(
    SpawnRate=&FloatFromCurve(
     FloatCurve=DataInterface() {propertyValues {Curve {keys [
         <at 0 0 RCIM_Constant>
         {
          arriveTangent=0.6761
          arriveTangentWeight=0.0188
          interpMode=RCIM_Cubic
          leaveTangent=0.6761
          leaveTangentWeight=1.3602
          time=2.1627
          value=4
         }
         {arriveTangent=-1.3671, interpMode=RCIM_Constant, tangentMode=RCTM_User, time=4.4374, value=3}
         <at 4.4521 0 RCIM_Constant>
        ]}}}
     CurveIndex=$System.Age
     Scale Curve=1.3
    )
    Use Spawn Probability=false
   ) {module=SpawnRate001}
   &SpawnBurst_Instantaneous(
    Spawn Count=&MakeCustomIntFromBool(Bool=&Boolean_OrOperation(Bool1=$User.bIsFrontEnd, Bool2=$User.bIsFrontEndPreview), True INT=0, False INT=1)
    Spawn Group=1
    Spawn Time=2.5
    Use Spawn Probability=false
    UseLoopCountLimit=false
   )
  ]
  particleSpawnScript [
   &InitializeParticle(
    Color=$System.TOD Color
    Lifetime=&MergeHelper_FloatSwitch(
     0=&RandomRangeFloat(Minimum=0.85, Maximum=1)
     1=3
     2=?
     3=?
     Index=&SpawnGroup()
     Num=::NewEnumerator0
    )
    Mass Mode=::NewEnumerator1
    Position Offset=<0 0 0.001>
    Sprite Rotation Angle=&UniformRangedFloat(Minimum=-180, Maximum=180)
    Sprite Rotation Mode=::NewEnumerator0
    Sprite Size=&Vector2DFromFloat(Value=6)
    Sprite Size Mode=::NewEnumerator4
    UsePositionOffset=true
    Write Material Random=true
   ) {module=InitializeParticle001}
   &InitialMeshOrientation(Mesh Orientation Mode=::NewEnumerator7, Use Rotation Vector=true)
   &self:ScratchModule(
    Offset End:Vector3f=<0 0 -50>
    Offset Start:Vector3f=<0 0 50>
    TracePosition:NiagaraPosition=&AddVectorToPosition(Position=/Niagara/DynamicInputs/Helpers/SimulationPosition(), Vector=<0 0 100>)
   ) {module=ScratchModule002}
   /SetVariables(Particles.VisibilityTag=&SpawnGroup()) {isSetParameters=true}
  ]
  particleUpdateScript [
   &ParticleState()
   &DynamicMaterialParameters(
    Index 0 Param 1=&FloatFromCurve(FloatCurve=DataInterface() {propertyValues {Curve {keys [<at 0 0.5>,<at 1 2>]}}})
    Module.Index0Float0Write=true
    Module.Index0Float1Write=true
    Module.Index0Float2Write=true
    Module.Index0Float3Write=true
   )
   &ScaleSpriteSize(Scale Factor=&Vector2DFromFloat(Value=&FloatFromCurve(
      FloatCurve=DataInterface() {propertyValues {Curve {keys [
          {arriveTangent=35.6763, interpMode=RCIM_Cubic, leaveTangent=3.7835, tangentMode=RCTM_User, time=0, value=0}
          {interpMode=RCIM_Cubic, leaveTangent=0.6903, time=1.0011, value=1}
         ]}}}
      Scale Curve=40
     )))
   &CameraOffset(Camera Offset Amount=20)
   &ScaleColor(
    Scale Alpha=&FloatFromCurve(
     FloatCurve=DataInterface() {propertyValues {Curve {keys [<at 0 1 RCIM_Cubic>, {interpMode=RCIM_Cubic, leaveTangent=-4.9981, time=1.0011, value=0.002}]}}}
     Scale Curve=0.6
    )
    ScaleA=true
    ScaleRGB=false
   )
   &SolveForcesAndVelocity(Clamp Velocity=false, Limit Acceleration=false)
   &ScaleColor(
    Scale Alpha=&Lerp_Float(
     A=1
     B=0.35
     Alpha=&FloatFromCurve(
      FloatCurve=DataInterface() {propertyValues {Curve {keys [<at 1 0>,<at 200 1>]}}}
      CurveIndex=&VectorLength(Vector=$Engine.Owner.Velocity)
     )
    )
    ScaleA=true
    ScaleRGB=true
   ) {module=ScaleColor001}
   &self:ScratchModule(
    Offset End:Vector3f=<0 0 -250>
    Offset Start:Vector3f=<0 0 250>
    SkipTrace:NiagaraBool=&IsPlatformSetActive(New Platform Set=DataInterface() {propertyValues {Platforms {qualityLevelMask=-29}}})
    TracePosition:NiagaraPosition=&AddVectorToPosition(Position=&SimulationPosition(), Vector=<0 0 100>)
   ) {module=ScratchModule}
   &UpdateMeshOrientation(
    Facing Direction=$Particles.CollisionNormal
    Facing Interpolation=false
    Orientation Method=::NewEnumerator1
    Override Quat=false
    TransformFacing=false
    TransformUp=false
   )
   &AlignSpriteToMeshOrientation(Mesh Orientation Relative Sprite Facing Vector:Vector3f=<0 0 1>, SetAlignment=true, SetFacing=true)
   &FN_Light_Attributes(
    Color=&ColorFromCurve(ColorCurve=DataInterface() {propertyValues {
       AlphaCurve {keys [<at 0 1>,<at 0.8661 1>,<at 1 0>]}
       BlueCurve {keys [<at 0 0 RCIM_Cubic>,<at 1 0.0003>]}
       GreenCurve {keys [<at 0 1.0901 RCIM_Cubic>,<at 1 0.0351>]}
       RedCurve {keys [<at 0 12 RCIM_Cubic>,<at 1 0.8229>]}
      }})
    Create Color Attribute=true
    Create Exponent Attribute=true
    Create Position Attribute=true
    Create Radius Attribute=true
    Enable Particle Light Enable=true
    Exponent=2
    Light Enabled=true
    Position=<0 0 10>
    Radius=&FloatFromCurve(
     FloatCurve=DataInterface() {propertyValues {Curve {keys [<at 0 1>,<at 3.223 1>,<at 3.2712 2>,<at 3.773 1>]}}}
     CurveIndex=$System.Age
     Scale Curve=80
    )
    Scale Color=&FloatFromCurve(
     FloatCurve=DataInterface() {propertyValues {Curve {keys [
         {arriveTangent=4.4475, time=0, value=0}
         {arriveTangent=5.3518, leaveTangent=-3.1774, time=0.1, value=1.2}
         {arriveTangent=-1.4972, leaveTangent=-1.4972, time=0.1523, value=0.7}
         {arriveTangent=-0.1667, leaveTangent=-0.1667, time=0.2052, value=1.1711}
         <at 0.258 0.7713>
         <at 0.2926 1.1775>
         {arriveTangent=-0.25, leaveTangent=-0.25, time=0.3608, value=0.689}
         <at 0.3865 1.193>
         <at 0.4263 0.8643>
         <at 0.4931 1.1191>
         {arriveTangent=-0.25, leaveTangent=-0.25, time=0.5428, value=0.7548}
         <at 0.5846 1.1063>
         <at 0.6571 0.8076>
         <at 0.6976 1.1685>
         {arriveTangent=0.1667, leaveTangent=0.1667, time=0.7542, value=3}
         <at 0.7906 3>
         <at 0.8535 0.8807>
         {arriveTangent=-3.3333, leaveTangent=-3.3333, time=0.9, value=1.2}
         <at 1 0>
        ]}}}
     Scale Curve=&Multiply_Float(A=3, B=&FNTimeOfDayScalar(Night Scale=0.5, Morning Scale=0.8, Evening Scale=0.8))
    )
   )
   &self:Slope_Position_01(
    Apply Position=true
    Apply Rotation=true
    Height Adjustment/Offset=1
    Slope Position:NiagaraPosition=$Engine.Emitter.SimulationPosition
    Trace Length=120
    Trace Origin Height=85
    bIsFrontEnd:NiagaraBool=$User.bIsFrontEndPreview
   ) {enabled=false, module=Slope_Position_01}
  ]
  rendererClasses [&NiagaraSpriteRendererProperties, &NiagaraLightRendererProperties]
  renderers [
   {rendererClass=&NiagaraSpriteRendererProperties}
   {rendererClass=&NiagaraSpriteRendererProperties, rendererIndex=1}
   {rendererClass=&NiagaraLightRendererProperties, rendererIndex=2}
  ]
 }
 SmokeGroundRings_Hand {
  emitterUpdateScript [
   &EmitterState()
   &SpawnBurst_Instantaneous(
    Spawn Count=&MakeCustomIntFromBool(Bool=&Boolean_OrOperation(Bool1=$User.bIsFrontEnd, Bool2=$User.bIsFrontEndPreview), True INT=0, False INT=1)
    Spawn Group=1
    Spawn Time=0.4
    Use Spawn Probability=false
    UseLoopCountLimit=false
   )
   &SpawnRate(
    SpawnRate=&FloatFromCurve(
     FloatCurve=DataInterface() {propertyValues {Curve {keys [<at 0.53 4>, {arriveTangent=-2.2759, interpMode=RCIM_Cubic, leaveTangent=-2.2759, time=2.2789, value=6}, <at 2.2876 0 RCIM_Constant>]}}}
     CurveIndex=$System.Age
     Scale Curve=1.3
    )
    Use Spawn Probability=false
   )
  ]
  particleSpawnScript [
   &InitializeParticle(
    Color=$System.TOD Color
    Lifetime=&MergeHelper_FloatSwitch(
     0=&RandomRangeFloat(Minimum=0.65, Maximum=0.85)
     1=2.8
     2=?
     3=?
     Index=&SpawnGroup()
     Num=::NewEnumerator0
    )
    Mass Mode=::NewEnumerator1
    Position Offset=<60 30 0.001>
    Sprite Rotation Angle=&UniformRangedFloat(Minimum=-180, Maximum=180)
    Sprite Rotation Mode=::NewEnumerator0
    Sprite Size=&Vector2DFromFloat(Value=3.5)
    Sprite Size Mode=::NewEnumerator4
    UsePositionOffset=true
    Write Material Random=true
   ) {module=InitializeParticle001}
   &InitialMeshOrientation(Mesh Orientation Mode=::NewEnumerator7, Use Rotation Vector=true)
   &self:ScratchModule(
    Offset End:Vector3f=<0 0 -50>
    Offset Start:Vector3f=<0 0 50>
    TracePosition:NiagaraPosition=&AddVectorToPosition(Position=&SimulationPosition(), Vector=<0 0 100>)
   ) {module=ScratchModule002}
   /SetVariables(Particles.VisibilityTag=&SpawnGroup()) {isSetParameters=true}
  ]
  particleUpdateScript [
   &ParticleState()
   &DynamicMaterialParameters(
    Index 0 Param 1=&FloatFromCurve(FloatCurve=DataInterface() {propertyValues {Curve {keys [<at 0 0.5>,<at 1 2>]}}})
    Module.Index0Float0Write=true
    Module.Index0Float1Write=true
    Module.Index0Float2Write=true
    Module.Index0Float3Write=true
   )
   &ScaleSpriteSize(
    Uniform Curve Index=$System.Age
    Uniform Curve Sprite Scale=DataInterface() {propertyValues {Curve {keys [{tangentMode=RCTM_User, time=0.53, value=0.8}, {tangentMode=RCTM_User, time=2.5425, value=1.2}]}}}
   ) {module=ScaleSpriteSize001}
   &ScaleSpriteSize(Scale Factor=&Vector2DFromFloat(Value=&FloatFromCurve(
      FloatCurve=DataInterface() {propertyValues {Curve {keys [
          {arriveTangent=35.6763, interpMode=RCIM_Cubic, leaveTangent=3.7835, tangentMode=RCTM_User, time=0, value=0}
          {interpMode=RCIM_Cubic, leaveTangent=0.6903, time=1.0011, value=1}
         ]}}}
      Scale Curve=40
     )))
   &CameraOffset(Camera Offset Amount=20)
   &ScaleColor(
    Scale Alpha=&FloatFromCurve(
     FloatCurve=DataInterface() {propertyValues {Curve {keys [<at 0 1 RCIM_Cubic>, {interpMode=RCIM_Cubic, leaveTangent=-4.9981, time=1.0011, value=0.002}]}}}
     Scale Curve=0.6
    )
    ScaleA=true
    ScaleRGB=false
   )
   &SolveForcesAndVelocity(Clamp Velocity=false, Limit Acceleration=false)
   &ScaleColor(
    Scale Alpha=&Lerp_Float(
     A=1
     B=0.35
     Alpha=&FloatFromCurve(FloatCurve=DataInterface() {propertyValues {Curve {keys [<at 1 0>,<at 200 1>]}}})
    )
    ScaleA=true
    ScaleRGB=true
   ) {module=ScaleColor001}
   &self:ScratchModule(
    Offset End:Vector3f=<0 0 -250>
    Offset Start:Vector3f=<0 0 250>
    SkipTrace:NiagaraBool=&IsPlatformSetActive(New Platform Set=DataInterface() {propertyValues {Platforms {qualityLevelMask=-29}}})
    TracePosition:NiagaraPosition=&AddVectorToPosition(Position=&SimulationPosition(), Vector=<0 0 100>)
   ) {module=ScratchModule}
   &UpdateMeshOrientation(
    Facing Direction=$Particles.CollisionNormal
    Facing Interpolation=false
    Orientation Method=::NewEnumerator1
    Override Quat=false
    TransformFacing=false
    TransformUp=false
   )
   &AlignSpriteToMeshOrientation(Mesh Orientation Relative Sprite Facing Vector:Vector3f=<0 0 1>, SetAlignment=true, SetFacing=true)
   &FN_Light_Attributes(
    Color=&ColorFromCurve(ColorCurve=DataInterface() {propertyValues {
       AlphaCurve {keys [<at 0 1>,<at 0.6097 1>,<at 1 0>]}
       BlueCurve {keys [<at 0 0 RCIM_Cubic>,<at 1 0.0003>]}
       GreenCurve {keys [<at 0 1.0901 RCIM_Cubic>,<at 1 0.0351>]}
       RedCurve {keys [<at 0 12 RCIM_Cubic>,<at 1 0.8229>]}
      }})
    Create Color Attribute=true
    Create Exponent Attribute=true
    Create Position Attribute=true
    Create Radius Attribute=true
    Enable Particle Light Enable=true
    Exponent=2
    Light Enabled=true
    Position=<60 30 10>
    Radius=50
    Scale Color=&FloatFromCurve(
     FloatCurve=DataInterface() {propertyValues {Curve {keys [
         {arriveTangent=4.4475, time=0, value=0}
         {arriveTangent=5.3518, leaveTangent=-3.1774, time=0.1, value=1.2}
         {arriveTangent=-1.4972, leaveTangent=-1.4972, time=0.1523, value=0.7}
         {arriveTangent=-0.1667, leaveTangent=-0.1667, time=0.2052, value=1.1711}
         <at 0.258 0.7713>
         <at 0.2926 1.1775>
         {arriveTangent=-0.25, leaveTangent=-0.25, time=0.3608, value=0.689}
         <at 0.3865 1.193>
         <at 0.4263 0.8643>
         <at 0.4931 1.1191>
         {arriveTangent=-0.25, leaveTangent=-0.25, time=0.5428, value=0.7548}
         <at 0.5846 1.1063>
         <at 0.6571 0.8076>
         <at 0.6976 1.1685>
         {arriveTangent=0.1667, leaveTangent=0.1667, time=0.7542, value=0.7713}
         <at 0.7906 1.1138>
         <at 0.8535 0.8807>
         {arriveTangent=-3.3333, leaveTangent=-3.3333, time=0.9, value=1.2}
         <at 1 0>
        ]}}}
     Scale Curve=&Multiply_Float(A=3, B=&FNTimeOfDayScalar(Night Scale=0.5, Morning Scale=0.8, Evening Scale=0.8))
    )
   )
   &self:Slope_Position_02(
    Apply Position=true
    Apply Rotation=true
    Height Adjustment/Offset=1
    Slope Position:NiagaraPosition=$Engine.Emitter.SimulationPosition
    Trace Length=120
    Trace Origin Height=85
    bIsFrontEnd:NiagaraBool=$User.bIsFrontEndPreview
   ) {enabled=false, module=Slope_Position_02}
  ]
  rendererClasses [&NiagaraSpriteRendererProperties, &NiagaraLightRendererProperties]
  renderers [
   {rendererClass=&NiagaraSpriteRendererProperties}
   {rendererClass=&NiagaraSpriteRendererProperties, rendererIndex=1}
   {rendererClass=&NiagaraLightRendererProperties, rendererIndex=2}
  ]
 }
}
summary {
 systemName=NS_Emote_AuroraOwl_GroundStuff
 emitters [
  {
   emitterName=SmokeGroundRings_Hand
   rendererClasses [&NiagaraSpriteRendererProperties, &NiagaraLightRendererProperties]
  }
  {
   emitterName=SmokeGroundRings_Feet
   rendererClasses [&NiagaraSpriteRendererProperties, &NiagaraLightRendererProperties]
  }
  {emitterName=DespawnBurst, rendererClasses [&NiagaraSpriteRendererProperties]}
  {emitterName=Glow_FE, rendererClasses [&NiagaraSpriteRendererProperties]}
 ]
 userVariables [
  User.bIsFrontEnd {type=NiagaraBool}
  User.bIsFrontEndPreview {type=NiagaraBool}
  User.bIsInWater {type=NiagaraBool}
  User.Flame_Size {type=NiagaraFloat, defaultValue=1, description=""}
  User.JetActiveCurve {type=NiagaraFloat, defaultValue=0, description=""}
  User.NormalizedNotifyProgress {type=NiagaraFloat, defaultValue=0, description=""}
 ]
}
