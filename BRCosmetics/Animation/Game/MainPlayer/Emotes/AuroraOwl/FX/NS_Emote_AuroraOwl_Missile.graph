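# Niagara effect for the AuroraOwl emote's missile: a missile with an early impact flare, followed about four seconds later by an explosion.
# Emitters are sequenced by loop delays, with Explosion_Flare, Explosion_Burst and GPU directional sparks supplying flash, shockwave, light, smoke and debris.
# Spawn groups select per-particle motion and set VisibilityTag, letting one emitter drive several sprite, mesh or light renderers.
# Explosion_Burst also kills particles based on owner speed.
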NiagaraSystem /BRCosmetics/Animation/Game/MainPlayer/Emotes/AuroraOwl/FX/NS_Emote_AuroraOwl_Missile
dependencies {
 usedDataInterfaces [
  &NiagaraDataInterfaceArrayFloat3
  &NiagaraDataInterfaceCurve
  &NiagaraDataInterfaceColorCurve
  &NiagaraDataInterfaceVectorCurve
  &NiagaraDataInterfaceVector2DCurve
  &NiagaraDataInterfaceArrayColor
  &NiagaraDataInterfacePlatformSet
  &NiagaraDataInterfaceArrayInt32
  &NiagaraDataInterfaceSkeletalMesh
 ]
 usedDynamicInputs [&NormalizedAngleToDegrees]
 usedModules [
  &SystemState
  &self:SystemSpawnScript.SystemScriptSource.SystemScriptGraph.NiagaraNodeAssignment_2.SetVariables_#1
  &self:SystemSpawnScript.SystemScriptSource.SystemScriptGraph.NiagaraNodeAssignment_4.SetVariables_#2
  &self:SystemSpawnScript.SystemScriptSource.SystemScriptGraph.NiagaraNodeAssignment_0.SetVariables_#3
 ]
}
emitters {
 DirectionalSparks_GPU {
  simTarget=GPUComputeSim
  emitterUpdateScript [
   &EmitterState(Life Cycle Mode=::NewEnumerator1, Loop Behavior=::NewEnumerator1, Loop Delay=4.04, Loop Duration=1.51, UseLoopDelay=true)
   &SpawnBurst_Instantaneous(Spawn Count=128, Use Spawn Probability=false, UseLoopCountLimit=false)
   &SpawnBurst_Instantaneous(Spawn Count=128, Spawn Time=1.51, Use Spawn Probability=false, UseLoopCountLimit=false) {module=SpawnBurst_Instantaneous001}
  ]
  particleSpawnScript [
   &PartitionParticles(Number Of Partitions=3, Partition Mode:ENiagara_PartitionMode_v2=::NewEnumerator1)
   &InitializeParticle(
    Color=<rgba 20 3.4136 0.0252 1>
    Lifetime=&RandomRangeFloat(Minimum=0.5, Maximum=0.8)
    Position Offset=<-20 -20 250>
    Sprite Size Max=<3 14>
    Sprite Size Min=<1 8>
    Sprite Size Mode=::NewEnumerator2
    UsePositionOffset=true
    Write Material Random=true
   )
   &AddVelocity(
    Velocity=&MergeHelper_VectorSwitch(0=<1 -1 0.7>, 1=<-0.5 0.5 1>, 2=<0.5 1 0.7>, 3=?, Index=$Particles.PartitionParticles.Partition, Num=::NewEnumerator1)
    Velocity Speed Scale=&RandomRangeFloat(Minimum=50, Maximum=250)
   )
   &ShapeLocation(MaskSpawnGroup=true, Non Uniform Scale=<10 10 1>, Override Seed=false, Sphere Radius=2, UseOwnerScale=true)
   /SetVariables(Particles.VisibilityTag=&SpawnGroup()) {isSetParameters=true}
  ]
  particleUpdateScript [
   &ParticleState()
   &GravityForce(Gravity=&VectorFromCurve(
     VectorCurve=DataInterface() {propertyValues {XCurve {keys [<at 0 0>,<at 1 0>]}, YCurve {keys [<at 0 0>,<at 1 0>]}, ZCurve {keys [<at 0 0>,<at 1 1>]}}}
     Scale Curve=<1 1 -500>
    ))
   &Drag(Drag=&RandomRangeFloat(Minimum=0, Maximum=1), Use Linear Drag=true, Use Rotational Drag=false)
   &SolveForcesAndVelocity(Clamp Velocity=false, Limit Acceleration=false)
  ]
  rendererClasses [&NiagaraSpriteRendererProperties]
  renderers [{rendererClass=&NiagaraSpriteRendererProperties}]
 }
 Explosion_Burst {
  emitterUpdateScript [
   &EmitterState(
    Inactive Response=::NewEnumerator2
    Life Cycle Mode=::NewEnumerator1
    Loop Behavior=::NewEnumerator1
    Loop Delay=3.9
    Loop Duration=3.9
    UseLoopDelay=true
   )
   &SpawnBurst_Instantaneous(Spawn Count=5, Spawn Group=4, Use Spawn Probability=false, UseLoopCountLimit=false) {module=SpawnBurst_Instantaneous005}
   &SpawnBurst_Instantaneous(Spawn Count=20, Spawn Group=3, Use Spawn Probability=false, UseLoopCountLimit=false) {module=SpawnBurst_Instantaneous004}
   &SpawnBurst_Instantaneous(Spawn Count=40, Spawn Group=2, Use Spawn Probability=false, UseLoopCountLimit=false) {module=SpawnBurst_Instantaneous002}
   &SpawnBurst_Instantaneous(Spawn Count=15, Spawn Group=5, Spawn Time=0.1, Use Spawn Probability=false, UseLoopCountLimit=false) {module=SpawnBurst_Instantaneous001}
   &SpawnBurst_Instantaneous(Spawn Count=3, Spawn Group=1, Spawn Time=0.05, Use Spawn Probability=false, UseLoopCountLimit=false) {module=SpawnBurst_Instantaneous003}
   &SpawnBurst_Instantaneous(Spawn Count=25, Use Spawn Probability=false)
  ]
  particleSpawnScript [
   &InitializeParticle(
    Color=&UniformRangedLinearColor(Minimum=<rgba 0.0521 0.0077 0 1>, Maximum=<rgba 0.0677 0.0475 0.0358 1>)
    Lifetime=&MergeHelper_FloatSwitch(
     0=&UniformRangedFloat(Minimum=1.35, Maximum=1.65)
     1=&RandomRangeFloat(Minimum=1, Maximum=1.3)
     2=&RandomRangeFloat(Minimum=0.85, Maximum=1)
     3=&RandomRangeFloat(Minimum=0.8, Maximum=1)
     4=&RandomRangeFloat(Minimum=1, Maximum=1.2)
     Index=&SpawnGroup()
     Num=::NewEnumerator3
    )
    Mass Mode=::NewEnumerator1
    Position Offset=&MergeHelper_VectorSwitch(0=<-20 -50 250>, 1=<-20 -50 250>, 2=?, 3=?, Index=&SpawnGroup(), Num=::NewEnumerator0)
    Sprite Rotation Mode=::NewEnumerator2
    Sprite Size Mode=::NewEnumerator3
    Uniform Sprite Size=&RandomRangeFloat(Minimum=50, Maximum=85)
    UsePositionOffset=true
    Write Material Random=true
   ) {module=InitializeParticle002}
   /SetVariables(Particles.InitSpriteRotRate=&UniformRangedFloat(Minimum=-3, Maximum=3)) {isSetParameters=true}
   &ShapeLocation(MaskSpawnGroup=true, Override Seed=false, Sphere Radius=20, Sphere Surface Distribution=1, UseOwnerScale=true)
   &AddVelocity(
    Velocity=&MergeHelper_VectorSwitch(
     0=&RandomRangeVector(Minimum=<-15 -15 -2>, Maximum=<15 15 10>)
     1=&RandomRangeVector(Minimum=<-5 -5 20>, Maximum=<5 5 27>)
     2=&RandomRangeVector(Minimum=<-15 -15 -2>, Maximum=<15 15 16>)
     3=&RandomRangeVector(Minimum=<-15 -15 -2>, Maximum=<15 15 16>)
     4=&RandomRangeVector(Minimum=<-15 -15 5>, Maximum=<15 15 5>)
     5=&RandomRangeVector(Minimum=<-8 -8 25>, Maximum=<8 8 40>)
     Index=&SpawnGroup()
     Num=::NewEnumerator4
    )
    Velocity Speed Scale=&MergeHelper_FloatSwitch(
     0=&RandomRangeFloat(Minimum=10, Maximum=18)
     1=&RandomRangeFloat(Minimum=15, Maximum=24)
     2=&RandomRangeFloat(Minimum=20, Maximum=30)
     3=&RandomRangeFloat(Minimum=20, Maximum=30)
     4=5
     5=&RandomRangeFloat(Minimum=10, Maximum=13)
     Index=&SpawnGroup()
     Num=::NewEnumerator4
    )
   )
   /SetVariables(Particles.VisibilityTag=&SpawnGroup()) {isSetParameters=true}
   /SetVariables(
    Particles.Debris Color:LinearColor=&LerpLinearColors(StartColor=<rgba 1 0.1142 0 1>, EndColor=<rgba 1 0.1142 0 1>, LerpFactor=1)
    Particles.Debris Sprite Size:Vector2f=&Vector2DFromFloat(Value=&RandomRangeFloat(Minimum=6, Maximum=10))
   ) {isSetParameters=true}
   /SetVariables(
    Particles.Smoke Streak Color:LinearColor=&UniformRangedLinearColor(Minimum=<rgba 1 0.1477 0 1>, Maximum=<rgba 0.6354 0.1091 0 1>)
    Particles.Smoke Streak Sprite Size:Vector2f=&UniformRangedVector2D(Minimum=<15 100>, Maximum=<20 80>)
   ) {isSetParameters=true}
   /SetVariables(
    Particles.Debris Color 2:LinearColor=&LerpLinearColors(StartColor=<rgba 1 0.1142 0 1>, EndColor=<rgba 1 0.1142 0 1>, LerpFactor=1)
    Particles.Debris Sprite Size 2:Vector2f=&Vector2DFromFloat(Value=&RandomRangeFloat(Minimum=20, Maximum=25))
   ) {isSetParameters=true}
   /SetVariables(Particles.Smoke Sprite Size:Vector2f=&Vector2DFromFloat(Value=&RandomRangeFloat(Minimum=50, Maximum=100))) {isSetParameters=true}
  ]
  particleUpdateScript [
   &KillParticles(Kill Particles=&SetBoolByFloatComparison(A=&VectorLength(Vector=$Engine.Owner.Velocity), B=70))
   &ParticleState()
   &SpriteRotationRate(Rotation Rate=&Multiply_Float(
     A=$Particles.InitSpriteRotRate
     B=&FloatFromCurve(
      FloatCurve=DataInterface() {propertyValues {Curve {keys [
          {arriveTangent=-4.0107, interpMode=RCIM_Cubic, leaveTangent=-4.0107, tangentMode=RCTM_User, time=-0.0046, value=1.0113}
          {arriveTangent=-0.2496, interpMode=RCIM_Cubic, leaveTangent=-0.2496, tangentMode=RCTM_User, time=0.5, value=0.16}
         ]}}}
      CurveIndex=$Emitter.Age
      Scale Curve=150
     )
    ))
   &ScaleColor(
    Scale Alpha=&FloatFromCurve(
     FloatCurve=DataInterface() {propertyValues {Curve {keys [
         <at 0.1219 4 RCIM_Cubic>
         {arriveTangent=-3.8779, interpMode=RCIM_Cubic, leaveTangent=-3.8779, time=0.194, value=2}
         {arriveTangent=-2.4936, interpMode=RCIM_Cubic, leaveTangent=-2.4936, time=0.8955, value=1}
         <at 0.9961 0 RCIM_Cubic>
        ]}}}
     Scale Curve=0.25
    )
    Scale RGB=&VectorFromFloat(Value=&FloatFromCurve(FloatCurve=DataInterface() {propertyValues {Curve {keys [{arriveTangent=1.8145, interpMode=RCIM_Cubic, leaveTangent=4.5795, tangentMode=RCTM_Break, time=0, value=0.4}, <at 0.2445 1>, <at 1 0.7>]}}}))
    ScaleA=true
    ScaleRGB=true
   )
   /Niagara/Modules/Update/Forces/CurlNoiseForce(
    Noise Frequency=0.0012
    Noise Strength=&MergeHelper_FloatSwitch(0=100, 1=0, 2=0, 3=0, 4=0, Index=&SpawnGroup(), Num=::NewEnumerator3)
    Pan Noise=true
    Pan Noise Field=<0 0 1>
   )
   &Drag(
    Drag=&MergeHelper_FloatSwitch(
     0=&FloatFromCurve(
      FloatCurve=DataInterface() {propertyValues {Curve {keys [
          {interpMode=RCIM_Cubic, leaveTangentWeight=0.9122, tangentMode=RCTM_Break, time=0, value=1}
          {arriveTangentWeight=0.9242, interpMode=RCIM_Cubic, tangentMode=RCTM_User, time=1, value=0}
         ]}}}
      Scale Curve=&Lerp_Float(A=2, B=4, Alpha=$Particles.Initial.MaterialRandom)
     )
     1=0
     2=1
     3=1
     4=0
     5=&FloatFromCurve(
      FloatCurve=DataInterface() {propertyValues {Curve {keys [<at 0 1>,<at 1 0>]}}}
      Scale Curve=&RandomRangeFloat(Minimum=3, Maximum=6)
     )
     Index=&SpawnGroup()
     Num=::NewEnumerator4
    )
    Use Linear Drag=true
    Use Rotational Drag=false
   )
   &GravityForce(Gravity=&MergeHelper_VectorSwitch(
     0=&VectorFromCurve(
      VectorCurve=DataInterface() {propertyValues {XCurve {keys [<at 0 0>,<at 1 0>]}, YCurve {keys [<at 0 0>,<at 1 0>]}, ZCurve {keys [<at 0 0>,<at 1 1>]}}}
      Scale Curve=&RandomRangeVector(Minimum=<0 0 50>, Maximum=<1 1 150>)
     )
     1=<0 0 -980>
     2=<0 0 -800>
     3=<0 0 -980>
     4=<0 0 0>
     Index=&SpawnGroup()
     Num=::NewEnumerator3
    ))
   &SolveForcesAndVelocity(Clamp Velocity=false, Limit Acceleration=false)
   &ScaleSpriteSize(Scale Factor=&Vector2DFromFloat(Value=&FloatFromCurve(FloatCurve=DataInterface() {propertyValues {Curve {keys [
          {arriveTangent=8.1876, interpMode=RCIM_Cubic, leaveTangent=5.3162, tangentMode=RCTM_User, time=0, value=0.15}
          {arriveTangent=1.0025, interpMode=RCIM_Cubic, leaveTangent=1.0025, tangentMode=RCTM_User, time=0.166, value=0.7682}
          {arriveTangent=0.2754, interpMode=RCIM_Cubic, leaveTangent=0.2754, tangentMode=RCTM_User, time=1.0463, value=1}
         ]}}})))
   &DynamicMaterialParameters(
    Index 0 Param 1=&Add_Float(
     A=&RandomRangeFloat(Minimum=0, Maximum=1)
     B=&FloatFromCurve(
      FloatCurve=DataInterface() {propertyValues {Curve {keys [
          {arriveTangent=2.7533, interpMode=RCIM_Cubic, leaveTangent=1.8958, tangentMode=RCTM_User, time=0, value=1}
          {arriveTangent=0.8292, interpMode=RCIM_Cubic, leaveTangent=0.8292, tangentMode=RCTM_User, time=0.321, value=1.3453}
          {arriveTangent=0.0529, interpMode=RCIM_Cubic, leaveTangent=0.2021, tangentMode=RCTM_User, time=1, value=1.5}
         ]}}}
      CurveIndex=&Add_Float(A=&Multiply_Float(A=0, B=0.5), B=$Particles.NormalizedAge)
      Scale Curve=&RandomRangeFloat(Minimum=0, Maximum=1)
     )
    )
    Index 0 Param 2=&FloatFromCurve(FloatCurve=DataInterface() {propertyValues {Curve {keys [
         {arriveTangent=0.0453, interpMode=RCIM_Cubic, tangentMode=RCTM_Break, time=0, value=-1}
         {arriveTangent=0.0604, interpMode=RCIM_Cubic, leaveTangent=0.0604, tangentMode=RCTM_User, time=1, value=-1}
        ]}}})
    Index 0 Param 3=&FloatFromCurve(
     FloatCurve=DataInterface() {propertyValues {Curve {keys [
         {arriveTangent=-2.3871, interpMode=RCIM_Cubic, leaveTangent=-2.3871, tangentMode=RCTM_User, time=0, value=10}
         {arriveTangent=-16.6667, interpMode=RCIM_Cubic, leaveTangent=-16.6667, time=0.0745, value=3}
         <at 0.6 0 RCIM_Cubic>
        ]}}}
     Scale Curve=&Lerp_Float(A=0.1, B=1, Alpha=1)
    )
    Index 0 Param 4=&FloatFromCurve(FloatCurve=DataInterface() {propertyValues {Curve {keys [<at 0.5 1>,<at 1 0>]}}})
    Index 1 Param 1=&FloatFromCurve(
     FloatCurve=DataInterface() {propertyValues {Curve {keys [
         {interpMode=RCIM_Cubic, leaveTangent=4.7088, tangentMode=RCTM_Break, time=0, value=0.5}
         {arriveTangent=0.0608, interpMode=RCIM_Cubic, leaveTangent=0.0608, tangentMode=RCTM_User, time=0.25, value=1}
        ]}}}
     Scale Curve=&RandomRangeFloat(Minimum=1, Maximum=1.5)
    )
    Index 1 Param 2=&FloatFromCurve(
     FloatCurve=DataInterface() {propertyValues {Curve {keys [<at 0 0.2>, {leaveTangent=0.9066, time=1, value=2}]}}}
     Scale Curve=&RandomRangeFloat(Minimum=0.1, Maximum=0.2)
    )
    Index 1 Param 3=&FloatFromCurve(FloatCurve=DataInterface() {propertyValues {Curve {keys [
         {interpMode=RCIM_Cubic, leaveTangent=6.3917, tangentMode=RCTM_Break, time=0, value=1}
         {arriveTangent=1.7549, interpMode=RCIM_Cubic, leaveTangent=10.6767, tangentMode=RCTM_User, time=0.25, value=1}
        ]}}})
    Index 1 Param 4=&FloatFromCurve(
     FloatCurve=DataInterface() {propertyValues {Curve {keys [{arriveTangent=-0.136, interpMode=RCIM_Cubic, leaveTangent=-0.136, tangentMode=RCTM_User, time=0, value=1}, <at 1 0.95>]}}}
     Scale Curve=&RandomRangeFloat(Minimum=1.2, Maximum=1.4)
    )
    Index 2 Param 1=&FloatFromCurve(FloatCurve=DataInterface() {propertyValues {Curve {keys [
         {interpMode=RCIM_Cubic, leaveTangent=1.874, tangentMode=RCTM_Break, time=0, value=0}
         {arriveTangent=0.8151, interpMode=RCIM_Cubic, leaveTangent=0.8151, tangentMode=RCTM_User, time=0.1776, value=0.75}
         <at 1 1>
        ]}}})
    Index 2 Param 2=&FloatFromCurve(
     FloatCurve=DataInterface() {propertyValues {Curve {keys [
         {arriveTangent=1.9056, interpMode=RCIM_Cubic, leaveTangent=1.9056, tangentMode=RCTM_User, time=0, value=0}
         {arriveTangent=0.0259, interpMode=RCIM_Cubic, leaveTangent=0.0259, tangentMode=RCTM_User, time=1, value=1.4}
        ]}}}
     Scale Curve=&RandomRangeFloat(Minimum=0, Maximum=1)
    )
    Index 2 Param 3=&FloatFromCurve(FloatCurve=DataInterface() {propertyValues {Curve {keys [
         {interpMode=RCIM_Cubic, leaveTangent=2.3064, tangentMode=RCTM_Break, time=0, value=0}
         {interpMode=RCIM_Cubic, leaveTangent=11.1233, tangentMode=RCTM_Break, time=1, value=2}
        ]}}})
    Index 2 Param 4=&FloatFromCurve(
     FloatCurve=DataInterface() {propertyValues {Curve {keys [
         {arriveTangent=-0.6867, interpMode=RCIM_Cubic, leaveTangent=-0.6867, tangentMode=RCTM_User, time=0, value=1}
         {arriveTangent=-0.2589, interpMode=RCIM_Cubic, leaveTangent=-0.2589, tangentMode=RCTM_User, time=0.3, value=0.8}
         {interpMode=RCIM_Cubic, leaveTangent=-1.2956, tangentMode=RCTM_Break, time=1, value=0.75}
        ]}}}
     Scale Curve=1.6
    )
    Module.Index0Float0Write=true
    Module.Index0Float1Write=true
    Module.Index0Float2Write=true
    Module.Index0Float3Write=true
    Module.Index1Float0Write=true
    Module.Index1Float1Write=true
    Module.Index1Float2Write=false
    Module.Index1Float3Write=true
    Module.Index2Float0Write=true
    Module.Index2Float1Write=true
    Module.Index2Float2Write=true
    Module.Index2Float3Write=true
    Write Parameter Index 1=::NewEnumerator1
    Write Parameter Index 2=::NewEnumerator1
   )
   &CameraOffset(Camera Offset Amount=15)
   /SetVariables(
    Particles.Sparks Color:LinearColor=<rgba 20 3.4136 0.0252 1>
    Particles.Sparks Size:Vector2f=&RandomRangeVector2D(Minimum=<0.5 8>, Maximum=<2 14>)
   ) {isSetParameters=true}
   /SetVariables(
    Particles.Debris Color:LinearColor=&MultiplyLinearColor(
     Linear Color=$Particles.Initial.Debris Color
     Scale Factor=&MakeVector4FromVectorAndFloat(
      "Vector (RGB)"=&VectorFromFloat(Value=&FloatFromCurve(FloatCurve=DataInterface() {propertyValues {Curve {keys [<at 0.1 1>,<at 0.4 0.1>,<at 1 0>]}}}, Scale Curve=2000))
      "Float (Alpha)"=&FloatFromCurve(FloatCurve=DataInterface() {propertyValues {Curve {keys [<at 0 0>,<at 0.3027 1>,<at 1 1>]}}})
     )
    )
    Particles.Debris Sprite Size:Vector2f=&Multiply_Vector2D(
     A=$Particles.Initial.Debris Sprite Size
     B=&Vector2DFromFloat(Value=&FloatFromCurve(FloatCurve=DataInterface() {propertyValues {Curve {keys [
           {interpMode=RCIM_Cubic, leaveTangentWeight=0.2874, tangentMode=RCTM_User, time=0, value=1}
           {arriveTangentWeight=0.4064, interpMode=RCIM_Cubic, leaveTangentWeight=0.065, tangentMode=RCTM_User, time=0.8621, value=1}
           <at 1 0 RCIM_Cubic>
          ]}}}))
    )
   ) {isSetParameters=true}
   /SetVariables(
    Particles.Smoke Streak Color:LinearColor=&MultiplyLinearColor(
     Linear Color=$Particles.Initial.Smoke Streak Color
     Scale Factor=&MakeVector4_FromLinearColor(LinearColor=&ColorFromCurve(ColorCurve=DataInterface() {propertyValues {
         AlphaCurve {keys [<at 0 1 RCIM_Cubic>,<at 1 0 RCIM_Cubic>]}
         BlueCurve {keys [<at 0 4.7917 RCIM_Cubic>,<at 1 1 RCIM_Cubic>]}
         GreenCurve {keys [<at 0 4.8426 RCIM_Cubic>,<at 1 1 RCIM_Cubic>]}
         RedCurve {keys [<at 0 5 RCIM_Cubic>,<at 1 1 RCIM_Cubic>]}
        }}))
    )
    Particles.Smoke Streak DynamicMaterialParameter1:Vector4f=&MakeVector4(
     X=&Multiply_Float(A=10, B=1)
     Y=0
     Z=0
     W=&FloatFromCurve(FloatCurve=DataInterface() {propertyValues {Curve {keys [<at 0 0.3>,<at 1 1>]}}})
    )
    Particles.Smoke Streak DynamicMaterialParameter2:Vector4f=&MakeVector4(
     X=&FloatFromCurve(
      FloatCurve=DataInterface() {propertyValues {Curve {keys [
          {interpMode=RCIM_Cubic, tangentMode=RCTM_User, time=0, value=0}
          {arriveTangent=0.6295, interpMode=RCIM_Cubic, leaveTangent=0.6295, tangentMode=RCTM_User, time=1, value=1}
         ]}}}
      Scale Curve=0.2
     )
     Y=&FloatFromCurve(FloatCurve=DataInterface() {propertyValues {Curve {keys [<at 0 1>,<at 1 0>]}}})
     Z=&FloatFromCurve(
      FloatCurve=DataInterface() {propertyValues {Curve {keys [{interpMode=RCIM_Cubic, tangentMode=RCTM_User, time=0.2, value=0.5}, {interpMode=RCIM_Cubic, tangentMode=RCTM_User, time=1, value=1}]}}}
      Scale Curve=2
     )
     W=&FloatFromCurve(FloatCurve=DataInterface() {propertyValues {Curve {keys [<at 0 0.3>,<at 1 0>]}}}, Scale Curve=-0.3)
    )
    Particles.Smoke Streak Sprite Size:Vector2f=&Multiply_Vector2D(
     A=$Particles.Initial.Smoke Streak Sprite Size
     B=&Vector2DFromCurve(
      Vector2Curve=DataInterface() {propertyValues {
        XCurve {keys [{arriveTangent=2.6663, interpMode=RCIM_Cubic, leaveTangent=2.6663, tangentMode=RCTM_User, time=0, value=0.3}, <at 1 1 RCIM_Cubic>]}
        YCurve {keys [{arriveTangent=2.9613, interpMode=RCIM_Cubic, leaveTangent=2.9613, tangentMode=RCTM_User, time=0, value=0.3}, <at 1 1 RCIM_Cubic>]}
       }}
      ScaleCurve=&Vector2DFromFloat(Value=2)
     )
    )
   ) {isSetParameters=true}
   /SetVariables(
    Particles.Debris Color 2:LinearColor=&MultiplyLinearColor(
     Linear Color=$Particles.Initial.Debris Color 2
     Scale Factor=&MakeVector4FromVectorAndFloat(
      "Vector (RGB)"=&VectorFromFloat(Value=&FloatFromCurve(FloatCurve=DataInterface() {propertyValues {Curve {keys [<at 0.1 1>,<at 0.4 0.1>,<at 1 0>]}}}, Scale Curve=2000))
      "Float (Alpha)"=&FloatFromCurve(FloatCurve=DataInterface() {propertyValues {Curve {keys [<at 0 0>,<at 0.3027 1>,<at 1 1>]}}})
     )
    )
    Particles.Debris Sprite Size 2:Vector2f=&Multiply_Vector2D(
     A=$Particles.Initial.Debris Sprite Size 2
     B=&Vector2DFromFloat(Value=&FloatFromCurve(FloatCurve=DataInterface() {propertyValues {Curve {keys [
           {interpMode=RCIM_Cubic, leaveTangentWeight=0.2874, tangentMode=RCTM_User, time=0, value=1}
           {arriveTangentWeight=0.4064, interpMode=RCIM_Cubic, leaveTangentWeight=0.065, tangentMode=RCTM_User, time=0.8621, value=1}
           <at 1 0 RCIM_Cubic>
          ]}}}))
    )
   ) {isSetParameters=true}
   /SetVariables(Particles.Smoke Sprite Size:Vector2f=&Multiply_Vector2D(
     A=$Particles.Initial.Smoke Sprite Size
     B=&Vector2DFromFloat(Value=&FloatFromCurve(FloatCurve=DataInterface() {propertyValues {Curve {keys [
           {arriveTangent=8.1876, interpMode=RCIM_Cubic, leaveTangent=5.3162, tangentMode=RCTM_User, time=0, value=0.15}
           {arriveTangent=1.0025, interpMode=RCIM_Cubic, leaveTangent=1.0025, tangentMode=RCTM_User, time=0.166, value=0.7682}
           {arriveTangent=0.2754, interpMode=RCIM_Cubic, leaveTangent=0.2754, tangentMode=RCTM_User, time=1.0463, value=1}
          ]}}}))
    )) {isSetParameters=true}
  ]
  rendererClasses [&NiagaraSpriteRendererProperties]
  renderers [
   {rendererClass=&NiagaraSpriteRendererProperties}
   {rendererClass=&NiagaraSpriteRendererProperties, rendererIndex=1}
   {rendererClass=&NiagaraSpriteRendererProperties, rendererIndex=2}
   {rendererClass=&NiagaraSpriteRendererProperties, rendererIndex=3}
   {rendererClass=&NiagaraSpriteRendererProperties, rendererIndex=4}
   {rendererClass=&NiagaraSpriteRendererProperties, rendererIndex=5}
  ]
 }
 Explosion_Flare {
  emitterUpdateScript [
   &EmitterState(Life Cycle Mode=::NewEnumerator1, Loop Behavior=::NewEnumerator1, Loop Delay=3.83, Loop Duration=4.79, UseLoopDelay=true)
   &SpawnBurst_Instantaneous(Spawn Group=5, Use Spawn Probability=false, UseLoopCountLimit=false) {module=SpawnBurst_Instantaneous005}
   &SpawnBurst_Instantaneous(Spawn Group=4, Spawn Time=0.1, Use Spawn Probability=false, UseLoopCountLimit=false) {module=SpawnBurst_Instantaneous004}
   &SpawnBurst_Instantaneous(Spawn Group=3, Use Spawn Probability=false, UseLoopCountLimit=false) {module=SpawnBurst_Instantaneous003}
   &SpawnBurst_Instantaneous(Spawn Group=2, Use Spawn Probability=false, UseLoopCountLimit=false)
   &SpawnBurst_Instantaneous(Spawn Group=1, Use Spawn Probability=false, UseLoopCountLimit=false) {module=SpawnBurst_Instantaneous001}
   &SpawnBurst_Instantaneous(Spawn Time=0.05, Use Spawn Probability=false, UseLoopCountLimit=false) {module=SpawnBurst_Instantaneous002}
  ]
  particleSpawnScript [
   &InitializeParticle(
    Color=<rgba 3 0.5301 0.3824 1>
    Lifetime=&MergeHelper_FloatSwitch(0=0.5, 1=0.5, 2=0.5, 3=0.15, 4=0.15, 5=0.5, Index=&SpawnGroup(), Num=::NewEnumerator4)
    Mesh Scale Mode=::NewEnumerator3
    Mesh Uniform Scale=0.2
    Position Offset=<-20 -50 260>
    Sprite Size Mode=::NewEnumerator3
    Uniform Sprite Size=230
    UsePositionOffset=true
    Write Material Random=true
   )
   &CameraOffset(Camera Offset Amount=70)
   /SetVariables(Particles.VisibilityTag=&SpawnGroup()) {isSetParameters=true}
   &InitialMeshOrientation(Mesh Orientation Mode=::NewEnumerator2, Orientation Vector=<0 0 1>, Use Rotation Vector=false)
   /SetVariables(
    Particles.Flare Color:LinearColor=<rgba 50 14.8271 5 5>
    Particles.Flare Sprite Size:Vector2f=&Vector2DFromFloat(Value=365)
   ) {isSetParameters=true}
   /SetVariables(
    Particles.Shockwave Color:LinearColor=<rgba 0.026 0.0208 0.0208 1>
    Particles.Shockwave Scale:Vector3f=&VectorFromFloat(Value=0.7)
   ) {isSetParameters=true}
   /SetVariables(
    Particles.Flare 2 Color:LinearColor=<rgba 10 0.7615 0 1>
    Particles.Flare 2 Sprite Size:Vector2f=&Vector2DFromFloat(Value=230)
   ) {isSetParameters=true}
  ]
  particleUpdateScript [
   &ParticleState()
   &ScaleSpriteSize(Uniform Curve Sprite Scale=DataInterface() {propertyValues {Curve {keys [
        <at 0 0 RCIM_Cubic>
        {arriveTangentWeight=0.1414, interpMode=RCIM_Cubic, leaveTangentWeight=0.33, tangentMode=RCTM_User, time=0.3, value=1}
        <at 1 1 RCIM_Cubic>
       ]}}})
   &ScaleMeshSize(Scale Factor=&VectorFromFloat(Value=&FloatFromCurve(FloatCurve=DataInterface() {propertyValues {Curve {keys [
          {arriveTangent=1.9878, interpMode=RCIM_Cubic, leaveTangent=1.9878, tangentMode=RCTM_User, time=0, value=0}
          <at 1.0465 1.3 RCIM_Cubic>
          <at 1.2 0 RCIM_Cubic>
         ]}}})))
   &ScaleColor(
    Scale RGB=&MakeVectorFromLinearColorRGB(Color=&ColorFromCurve(ColorCurve=DataInterface() {propertyValues {
        AlphaCurve {keys [<at 0 1>,<at 0.3925 1>,<at 1 0>]}
        BlueCurve {keys [<at 0 0.25>,<at 0.368 0>,<at 0.6248 0.0698>,<at 1 0.037>]}
        GreenCurve {keys [<at 0 2.8536>,<at 0.368 1.8026>,<at 0.6248 0.1719>,<at 1 0.0811>]}
        RedCurve {keys [<at 0 5>,<at 0.368 5>,<at 0.6248 0.6979>,<at 1 0.3698>]}
       }}))
    ScaleA=true
    ScaleRGB=true
   )
   /SetVariables(Particles.Glow Size:Vector3f=&VectorFromFloat(Value=&FloatFromCurve(FloatCurve=DataInterface() {propertyValues {Curve {keys [<at 0 0.4>,<at 1 1.25>]}}}, Scale Curve=1.25))) {isSetParameters=true}
   &DynamicMaterialParameters(
    Index 0 Param 1=&Lerp_Float(
     A=1
     B=&FloatFromCurve(FloatCurve=DataInterface() {propertyValues {Curve {keys [<at 0 1>,<at 0.3172 1>,<at 1 1>]}}})
     Alpha=&FN_PickFloatDependOnPlatform(TargetPlatformActive=1, TargetPlatformNoActive=0)
    )
    Index 0 Param 2=&FloatFromCurve(FloatCurve=DataInterface() {propertyValues {Curve {keys [<at 0 15>,<at 1 0>]}}})
    Module.Index0Float0Write=true
    Module.Index0Float1Write=true
    Module.Index0Float2Write=true
    Module.Index0Float3Write=true
   )
   /SetVariables(
    Particles.Flare Color:LinearColor=&MakeLinearColorFromVectorAndFloat(
     "Vector (RGB)"=&MakeVectorFromLinearColorRGB(Color=$Particles.Initial.Flare Color)
     "Float (Alpha)"=&FloatFromCurve(FloatCurve=DataInterface() {propertyValues {Curve {keys [<at 0 1>,<at 1 0>]}}})
    )
    Particles.Flare DynamicMaterialParameter_01:Vector4f=&MakeVector4(X=1, Y=1, Z=1, W=1)
    Particles.Flare Sprite Size:Vector2f=&Multiply_Vector2D(
     A=$Particles.Initial.Flare Sprite Size
     B=&Vector2DFromCurve(
      Vector2Curve=DataInterface() {propertyValues {
        XCurve {keys [<at 0 0.1>,<at 1 1.8>]}
        YCurve {keys [<at 0 0.1>, {arriveTangent=-0.1, interpMode=RCIM_Cubic, leaveTangent=-0.1, time=0.65, value=0.9332}, <at 1 0 RCIM_Cubic>]}
       }}
      ScaleCurve=<1 1>
     )
    )
   ) {isSetParameters=true}
   /SetVariables(
    Particles.Shockwave Color:LinearColor=&MakeLinearColorFromVectorAndFloat(
     "Vector (RGB)"=&MakeVectorFromLinearColorRGB(Color=$Particles.Initial.Shockwave Color)
     "Float (Alpha)"=&FloatFromCurve(FloatCurve=DataInterface() {propertyValues {Curve {keys [<at 0 1>,<at 1 0.5>]}}})
    )
    Particles.Shockwave Scale:Vector3f=&Multiply_Vector(
     A=$Particles.Initial.Shockwave Scale
     B=&VectorFromFloat(Value=&FloatFromCurve(FloatCurve=DataInterface() {propertyValues {Curve {keys [<at 0 1>,<at 1 8>]}}}))
    )
   ) {isSetParameters=true}
   /SetVariables(
    Particles.Flare 2 Color:LinearColor=&MultiplyLinearColor(
     Linear Color=$Particles.Initial.Flare 2 Color
     Scale Factor=&MakeVector4FromVectorAndFloat(
      "Vector (RGB)"=&MakeVectorFromLinearColorRGB(Color=<rgba 1 1 1 1>)
      "Float (Alpha)"=&FloatFromCurve(FloatCurve=DataInterface() {propertyValues {Curve {keys [<at 0 1>,<at 1 0>]}}})
     )
    )
    Particles.Flare 2 Sprite Size:Vector2f=&Multiply_Vector2D(
     A=$Particles.Initial.Flare 2 Sprite Size
     B=&Vector2DFromFloat(Value=&FloatFromCurve(FloatCurve=DataInterface() {propertyValues {Curve {keys [
           <at 0 0.2 RCIM_Cubic>
           {arriveTangentWeight=0.1022, interpMode=RCIM_Cubic, leaveTangentWeight=0.1941, tangentMode=RCTM_User, time=0.1335, value=1}
           {arriveTangentWeight=0.1292, interpMode=RCIM_Cubic, leaveTangentWeight=0.3119, tangentMode=RCTM_User, time=0.3873, value=1}
           <at 1 0 RCIM_Cubic>
          ]}}}))
    )
   ) {isSetParameters=true}
   &FN_Light_Attributes(
    Color=&ColorFromCurve(ColorCurve=DataInterface() {propertyValues {
       AlphaCurve {keys [<at 0 1>,<at 1 0>]}
       BlueCurve {keys [<at 0 0 RCIM_Cubic>,<at 1 0.0003>]}
       GreenCurve {keys [<at 0 1.0901 RCIM_Cubic>,<at 1 0.0351>]}
       RedCurve {keys [<at 0 12 RCIM_Cubic>,<at 1 0.8229>]}
      }})
    Create Color Attribute=true
    Create Exponent Attribute=true
    Create Radius Attribute=true
    Enable Particle Light Enable=true
    Exponent=4
    Light Enabled=true
    Radius=250
    Scale Color=10
   )
  ]
  rendererClasses [
   &NiagaraSpriteRendererProperties
   &NiagaraMeshRendererProperties
   &NiagaraLightRendererProperties
  ]
  renderers [
   {rendererClass=&NiagaraSpriteRendererProperties}
   {rendererClass=&NiagaraMeshRendererProperties, rendererIndex=1}
   {rendererClass=&NiagaraMeshRendererProperties, rendererIndex=2}
   {rendererClass=&NiagaraSpriteRendererProperties, rendererIndex=3}
   {rendererClass=&NiagaraMeshRendererProperties, rendererIndex=4}
   {rendererClass=&NiagaraSpriteRendererProperties, rendererIndex=5}
   {rendererClass=&NiagaraLightRendererProperties, rendererIndex=6}
  ]
 }
 ImpactFlare {
  emitterUpdateScript [
   &EmitterState(Life Cycle Mode=::NewEnumerator1, Loop Behavior=::NewEnumerator1, Loop Delay=0.05, Loop Duration=1, UseLoopDelay=true)
   &SpawnBurst_Instantaneous(Use Spawn Probability=false, UseLoopCountLimit=false) {module=SpawnBurst_Instantaneous003}
   &SpawnBurst_Instantaneous(Spawn Group=1, Use Spawn Probability=false, UseLoopCountLimit=false) {module=SpawnBurst_Instantaneous001}
   &SpawnBurst_Instantaneous(Spawn Group=2, Use Spawn Probability=false, UseLoopCountLimit=false)
  ]
  particleSpawnScript [
   /SetVariables(Particles.VisibilityTag=&SpawnGroup(Max Spawn Group=2)) {isSetParameters=true}
   &InitializeParticle(
    Color=&MakeCustomLinearColorFromBool(
     Bool=&SetBoolByIntComparison(A=&SpawnGroup(), B=1, Comparison Type=::NewEnumerator7)
     True Color=&SelectLinearColorFromArray(
      Array Sampling Mode=::NewEnumerator1
      Direct Array Index=&SpawnGroup()
      Color Selection Array=DataInterface() {propertyValues {ColorData [<rgba 1 1 1 1>, <rgba 1 1 1 0.5>, <rgba 1 1 1 1>, <rgba 100 61.7074 35.1155 1>]}}
      Random Seed=?
     )
     False Color=<rgba 1 1 1 1>
    )
    Lifetime=&MaskFloatBySpawnGroup(
     Masked Float=0.5
     Passthrough Float=&MaskFloatBySpawnGroup(Masked Float=0.25, Passthrough Float=0.25, Spawn Group Mask=1)
    )
    Mesh Scale Mode=::NewEnumerator3
    Mesh Uniform Scale=&MaskFloatBySpawnGroup(Masked Float=1, Passthrough Float=1, Spawn Group Mask=2)
    Position Offset=<40 -20 150>
    Sprite Rotation Angle=&MaskFloatBySpawnGroup(Masked Float=0, Passthrough Float=&RandomRangeFloat(Minimum=0, Maximum=1))
    Sprite Rotation Mode=::NewEnumerator1
    Sprite Size=&Vector2DFromFloat(Value=250)
    Sprite Size Mode=::NewEnumerator4
    UsePositionOffset=true
    Write Material Random=true
   )
   &CameraOffset(Camera Offset Amount=&MaskFloatBySpawnGroup(Masked Float=50, Passthrough Float=80))
  ]
  particleUpdateScript [
   &ParticleState()
   &SpriteRotationRate(
    Delta Time:NiagaraFloat=&FloatFromCurve(FloatCurve=DataInterface() {propertyValues {Curve {keys [
         {tangentMode=RCTM_User, time=0, value=1}
         {arriveTangent=-0.3174, interpMode=RCIM_Cubic, leaveTangent=-0.3174, tangentMode=RCTM_User, time=0.3, value=1}
         {arriveTangent=0.0123, interpMode=RCIM_Cubic, leaveTangent=0.0123, tangentMode=RCTM_User, time=1, value=0}
        ]}}})
    Rotation Rate=&MaskFloatBySpawnGroup(Masked Float=5, Passthrough Float=0, Spawn Group Mask=1)
   )
   &ScaleSpriteSize(
    Scale Factor=&MaskVector2dBySpawnGroup(
     Masked Vector2D=&Vector2DFromCurve(
      Vector2Curve=DataInterface() {propertyValues {
        XCurve {keys [<at 0 1 RCIM_Cubic>, {arriveTangent=-3.1442, interpMode=RCIM_Cubic, leaveTangent=-3.1442, tangentMode=RCTM_User, time=1, value=0}]}
        YCurve {keys [
          {arriveTangent=4.8145, interpMode=RCIM_Cubic, time=0, value=1}
          {arriveTangent=-3.0695, interpMode=RCIM_Cubic, leaveTangent=-3.0695, tangentMode=RCTM_User, time=1, value=0}
         ]}
       }}
      ScaleCurve=<1 1>
     )
     Passthrough Vector2D=&Vector2DFromFloat(Value=&MaskFloatBySpawnGroup(
       Masked Float=&FloatFromCurve(FloatCurve=DataInterface() {propertyValues {Curve {keys [
            {arriveTangent=5.1291, interpMode=RCIM_Cubic, leaveTangent=5.1291, tangentMode=RCTM_User, time=0, value=0}
            {arriveTangentWeight=0.1274, interpMode=RCIM_Cubic, leaveTangentWeight=0.7778, tangentMode=RCTM_User, time=0.5, value=1}
           ]}}})
       Passthrough Float=&FloatFromCurve(FloatCurve=DataInterface() {propertyValues {Curve {keys [
            {arriveTangent=3.3451, interpMode=RCIM_Cubic, leaveTangent=3.3451, tangentMode=RCTM_User, time=0, value=0.5}
            {arriveTangentWeight=0.1274, interpMode=RCIM_Cubic, leaveTangentWeight=0.7778, tangentMode=RCTM_User, time=0.5, value=1}
           ]}}})
       Spawn Group Mask=1
      ))
     Spawn Group Mask=0
    )
    Scale Sprite Size Mode=::NewEnumerator2
   )
   &Color(Color=&MakeCustomLinearColorFromBool(
     Bool=&SetBoolByIntComparison(A=&SpawnGroup(), B=1, Comparison Type=::NewEnumerator7)
     True Color=&SelectLinearColorFromArray(
      Array Sampling Mode=::NewEnumerator1
      Direct Array Index=&SpawnGroup()
      Color Selection Array=DataInterface() {propertyValues {ColorData [<rgba 1 1 1 1>, <rgba 1 1 1 0.5>, <rgba 1 1 1 1>, <rgba 100 61.7074 35.1155 1>]}}
      Random Seed=?
     )
     False Color=<rgba 1 1 1 1>
    ))
   &ScaleMeshSize(Scale Factor=&VectorFromFloat(Value=&MaskFloatBySpawnGroup(
      Masked Float=&FloatFromCurve(FloatCurve=DataInterface() {propertyValues {Curve {
          defaultValue=0
          keys [
           {arriveTangent=4.1785, interpMode=RCIM_Cubic, leaveTangent=4.1785, tangentMode=RCTM_User, time=0, value=0}
           {arriveTangent=1.529, interpMode=RCIM_Cubic, leaveTangent=1.529, tangentMode=RCTM_User, time=0.2042, value=0.6228}
           {arriveTangent=0.0224, interpMode=RCIM_Cubic, leaveTangent=0.0224, tangentMode=RCTM_User, time=1, value=1}
          ]
         }}})
      Passthrough Float=&FloatFromCurve(FloatCurve=DataInterface() {propertyValues {Curve {
          defaultValue=0
          keys [
           <at 0 0 RCIM_Cubic>
           {arriveTangent=1.6664, interpMode=RCIM_Cubic, leaveTangent=1.6664, tangentMode=RCTM_User, time=0.45, value=0.2}
           {arriveTangent=1.1179, interpMode=RCIM_Cubic, leaveTangent=1.1179, tangentMode=RCTM_User, time=0.55, value=0.75}
           {arriveTangent=0.0224, interpMode=RCIM_Cubic, leaveTangent=0.0224, tangentMode=RCTM_User, time=1, value=1}
          ]
         }}})
      Spawn Group Mask=2
     )))
   &DynamicMaterialParameters(
    Index 0 Param 1=&MaskFloatBySpawnGroup(
     Masked Float=&FloatFromCurve(FloatCurve=DataInterface() {propertyValues {Curve {keys [{arriveTangent=0.1648, interpMode=RCIM_Cubic, leaveTangent=0.1648, tangentMode=RCTM_User, time=0, value=0.9}, <at 0.5 1 RCIM_Cubic>]}}})
     Passthrough Float=&FloatFromCurve(FloatCurve=DataInterface() {propertyValues {Curve {keys [<at 0 0.9 RCIM_Cubic>,<at 1 1 RCIM_Cubic>]}}})
     Spawn Group Mask=2
    )
    Index 0 Param 3=5
    Index 0 Param 4=0.04
    Module.Index0Float0Write=true
    Module.Index0Float1Write=false
    Module.Index0Float2Write=true
    Module.Index0Float3Write=true
    Module.Index1Float0Write=true
    Module.Index1Float1Write=true
    Module.Index1Float2Write=true
    Module.Index1Float3Write=true
    Module.Index2Float0Write=true
    Module.Index2Float1Write=true
    Module.Index2Float2Write=true
    Module.Index2Float3Write=true
    Write Parameter Index 1=::NewEnumerator1
    Write Parameter Index 2=::NewEnumerator1
   )
   &ScaleColor(
    Scale Alpha=&MaskFloatBySpawnGroup(
     Masked Float=&FloatFromCurve(FloatCurve=DataInterface() {propertyValues {Curve {keys [
          {arriveTangent=-23.168, interpMode=RCIM_Cubic, leaveTangent=-23.168, tangentMode=RCTM_User, time=0, value=4}
          {arriveTangent=-4.5789, interpMode=RCIM_Cubic, leaveTangent=-4.5789, tangentMode=RCTM_User, time=0.2, value=1}
          <at 1 0 RCIM_Cubic>
         ]}}})
     Passthrough Float=&MaskFloatBySpawnGroup(
      Masked Float=&FloatFromCurve(
       FloatCurve=DataInterface() {propertyValues {Curve {
          defaultValue=0
          keys [
           {arriveTangent=9.0725, interpMode=RCIM_Cubic, leaveTangent=9.0725, tangentMode=RCTM_User, time=0, value=0}
           {arriveTangent=-2.5249, interpMode=RCIM_Cubic, leaveTangent=-2.5249, tangentMode=RCTM_User, time=0.1408, value=0.9847}
           {arriveTangent=-0.0405, interpMode=RCIM_Cubic, leaveTangent=-0.0405, tangentMode=RCTM_User, time=1, value=0}
          ]
         }}}
       Scale Curve=0.2
      )
      Passthrough Float=&MaskFloatBySpawnGroup(
       Masked Float=&FloatFromCurve(
        FloatCurve=DataInterface() {propertyValues {Curve {keys [<at 0 1 RCIM_Cubic>,<at 1 0 RCIM_Cubic>]}}}
        Scale Curve=&MaskFloatBySpawnGroup(Masked Float=0.2, Passthrough Float=1, Spawn Group Mask=2)
       )
       Passthrough Float=&FloatFromCurve(
        FloatCurve=DataInterface() {propertyValues {Curve {keys [<at 0 1 RCIM_Cubic>,<at 1 0 RCIM_Cubic>]}}}
        Scale Curve=&MaskFloatBySpawnGroup(Masked Float=0.2, Passthrough Float=1, Spawn Group Mask=2)
       )
       Spawn Group Mask=2
      )
      Spawn Group Mask=1
     )
    )
    Scale RGB=&MakeVectorFromLinearColorRGB(Color=&MakeCustomLinearColorFromBool(
      Bool=&SetBoolByIntComparison(A=&SpawnGroup(), B=1, Comparison Type=::NewEnumerator7)
      True Color=&ColorFromCurve(ColorCurve=DataInterface() {propertyValues {
         AlphaCurve {keys []}
         BlueCurve {keys [{arriveTangent=-1.5872, interpMode=RCIM_Cubic, leaveTangent=-1.5872, tangentMode=RCTM_User, time=0, value=3}, <at 1 0.5 RCIM_Cubic>]}
         GreenCurve {keys [{arriveTangent=-4.9882, interpMode=RCIM_Cubic, leaveTangent=-4.9882, tangentMode=RCTM_User, time=0, value=3.75}, <at 1 0.625 RCIM_Cubic>]}
         RedCurve {keys [{arriveTangent=-3.5785, interpMode=RCIM_Cubic, leaveTangent=-3.5785, tangentMode=RCTM_User, time=0, value=4}, <at 1 1 RCIM_Cubic>]}
        }})
      False Color=<rgba 1 1 1 1>
     ))
    ScaleA=true
    ScaleRGB=true
   )
  ]
  rendererClasses [
   &NiagaraSpriteRendererProperties
   &NiagaraLightRendererProperties
   &NiagaraMeshRendererProperties
  ]
  renderers [
   {rendererClass=&NiagaraSpriteRendererProperties}
   {rendererClass=&NiagaraLightRendererProperties, rendererIndex=1}
   {rendererClass=&NiagaraMeshRendererProperties, rendererIndex=2}
  ]
 }
 Missile_FX {
  emitterUpdateScript [
   &EmitterState(Life Cycle Mode=::NewEnumerator1, Loop Behavior=::NewEnumerator1, Loop Duration=3.6, UseLoopDelay=false)
   &SpawnBurst_Instantaneous(Spawn Group=2, Use Spawn Probability=false, UseLoopCountLimit=false) {module=SpawnBurst_Instantaneous002}
   &SpawnBurst_Instantaneous(Spawn Count=10, Spawn Group=1, Use Spawn Probability=false, UseLoopCountLimit=false) {module=SpawnBurst_Instantaneous001}
   &SpawnBurst_Instantaneous(Spawn Count=1, Use Spawn Probability=false)
   /SetVariables(
    Emitter.Flame1A:LinearColor=<rgba 0.125 0.0688 0.0125 1>
    Emitter.Flame1B:LinearColor=<rgba 0.5 0.4333 0.1 1>
    Emitter.Flame2A:LinearColor=<rgba 1 0.6667 0.2 1>
    Emitter.Flame2B:LinearColor=<rgba 0.4 0.2133 0.08 1>
    Emitter.Flame3A:LinearColor=<rgba 1 0.4838 0.1 1>
    Emitter.Flame3B:LinearColor=<rgba 0.4583 0.2796 0.0458 1>
    Emitter.FlameCoreA:LinearColor=<rgba 1 0.9088 0.526 1>
    Emitter.FlameCoreB:LinearColor=<rgba 1 0.7396 0.2188 1>
    Emitter.FlameDark:LinearColor=<rgba 0.1927 0.1172 0.0416 1>
    Emitter.FlameLight:LinearColor=<rgba 0.3854 0.2318 0.0781 1>
   ) {isSetParameters=true}
  ]
  particleSpawnScript [
   &InitializeParticle(
    Color=<rgba 1 0.4304 0.2426 1>
    Lifetime=&MergeHelper_FloatSwitch(0=3.515, 1=0.18, 2=3.515, 3=?, Index=&SpawnGroup(), Num=::NewEnumerator1)
    Mesh Scale=<0.5 0.5 0.6>
    Mesh Scale Mode=::NewEnumerator4
    Sprite Size Mode=::NewEnumerator3
    Uniform Sprite Size=220
    UsePositionOffset=false
    Write Material Random=true
   )
   &InitialMeshOrientation(
    Rotation=&MergeHelper_VectorSwitch(0=<0.5 0 0>, 1=<0 0.75 0>, 2=?, 3=?, Index=&SpawnGroup(), Num=::NewEnumerator0)
    Rotation Coordinate Space=::NewEnumerator0
    Use Orientation Vector=true
    Use Rotation Vector=true
   ) {module=InitialMeshOrientation001}
   /SetVariables(
    Particles.Flame2Scale:Vector3f=&Multiply_Vector(A=$Particles.Scale, B=<1 1 1>)
    Particles.FlameCoreScale:Vector3f=&Multiply_Vector(A=$Particles.Scale, B=<0.65 0.6 1>)
   ) {isSetParameters=true}
   /SetVariables(Particles.VisibilityTag=&SpawnGroup()) {isSetParameters=true}
   &CameraOffset(Camera Offset Amount=&MergeHelper_FloatSwitch(0=0, 1=0, 2=20, 3=?, Index=&SpawnGroup(), Num=::NewEnumerator1))
   /SetVariables(Particles.Flare Color:LinearColor=<rgba 3 1.0965 0.4688 1>, Particles.Flare Sprite Size:Vector2f=<120 120>) {isSetParameters=true}
  ]
  particleUpdateScript [
   &ParticleState()
   &SolveForcesAndVelocity(Clamp Velocity=false, Limit Acceleration=false)
   &SkeletalMeshLocation(
    Apply Sampled Orientation to Mesh Particles=true
    Bone Sampling Mode=::NewEnumerator4
    Inherit Velocity=false
    MaskSpawnGroup=false
    Sampled Position Offset=&SelectVectorFromArray(
     Direct Array Index=&SpawnGroup()
     Vector Selection Array=DataInterface() {propertyValues {FloatData [<0 0 -30>, <0 0 20>, <0 0 -40>]}}
    )
    Sampled Position Offset Coordinate Space:ENiagaraMeshTransforms=::NewEnumerator3
    Skeletal Mesh=$System.Skeletal Mesh
   ) {module=SkeletalMeshLocation001}
   &ScaleColor(
    Scale Alpha=&FloatFromCurve(FloatCurve=DataInterface() {propertyValues {Curve {keys [
         <at 0 1>
         <at 0.05 0.1>
         {arriveTangent=1.5287, leaveTangent=1.5287, time=0.1, value=1}
         <at 0.1808 0.3>
         {arriveTangent=-1.1287, leaveTangent=-1.1287, time=0.241, value=1}
         <at 0.3137 0.15>
         {arriveTangent=-0.4335, arriveTangentWeight=0.2916, leaveTangent=-0.4335, leaveTangentWeight=0.1147, time=0.3571, value=1}
         <at 0.4291 0.1>
         {arriveTangent=-0.7251, arriveTangentWeight=0.1918, leaveTangent=-0.7251, leaveTangentWeight=0.058, time=0.5086, value=1}
         <at 0.567 0>
         <at 0.6229 1>
         <at 0.6808 0>
         <at 0.7553 1>
         <at 0.8163 0>
         {arriveTangentWeight=0.0138, leaveTangentWeight=0.0174, time=0.8577, value=1}
         <at 0.9072 0>
         {arriveTangentWeight=0.0841, leaveTangentWeight=0.151, time=0.9394, value=1}
         {arriveTangentWeight=0.198, leaveTangentWeight=0.1375, time=0.9536, value=1}
         <at 1 0>
        ]}}})
    Scale RGB=&MakeVectorFromLinearColorRGB(Color=&LerpLinearColors(StartColor=<rgba 2 2 2 1>, EndColor=<rgba 1 1 1 1>, LerpFactor=$System.Timeline.Time))
    ScaleA=true
    ScaleRGB=true
   )
   /SetVariables(
    Particles.Flame2Color:LinearColor=&MakeLinearColorFromVectorAndFloat(
     "Vector (RGB)"=<1 1 1>
     "Float (Alpha)"=&FloatFromCurve(FloatCurve=DataInterface() {propertyValues {Curve {keys [
          <at 0 1>
          <at 0.05 0.1>
          {arriveTangent=1.5287, leaveTangent=1.5287, time=0.1, value=1}
          <at 0.1808 0.3>
          {arriveTangent=-1.1287, leaveTangent=-1.1287, time=0.241, value=1}
          <at 0.3137 0.15>
          {arriveTangent=-0.4335, arriveTangentWeight=0.2916, leaveTangent=-0.4335, leaveTangentWeight=0.1147, time=0.3571, value=1}
          <at 0.4291 0.1>
          {arriveTangent=-0.7251, arriveTangentWeight=0.1918, leaveTangent=-0.7251, leaveTangentWeight=0.058, time=0.5086, value=1}
          <at 0.567 0>
          <at 0.6229 1>
          <at 0.6808 0>
          <at 0.7553 1>
          <at 0.8163 0>
          {arriveTangentWeight=0.0138, leaveTangentWeight=0.0174, time=0.8577, value=1}
          <at 0.9072 0>
          {arriveTangentWeight=0.0841, leaveTangentWeight=0.151, time=0.9394, value=1}
          {arriveTangentWeight=0.198, leaveTangentWeight=0.1375, time=0.9536, value=1}
          <at 1 0>
         ]}}})
    )
    Particles.FlameCoreColor:LinearColor=&MakeLinearColorFromVectorAndFloat(
     "Vector (RGB)"=&MakeVectorFromLinearColorRGB(Color=<rgba 1 0.0473 0 1>)
     "Float (Alpha)"=&FloatFromCurve(
      FloatCurve=DataInterface() {propertyValues {Curve {keys [
          <at 0 1>
          <at 0.05 0.1>
          {arriveTangent=1.5287, leaveTangent=1.5287, time=0.1, value=1}
          <at 0.1808 0.3>
          {arriveTangent=-1.1287, leaveTangent=-1.1287, time=0.241, value=1}
          <at 0.3137 0.15>
          {arriveTangent=-0.4335, arriveTangentWeight=0.2916, leaveTangent=-0.4335, leaveTangentWeight=0.1147, time=0.3571, value=1}
          <at 0.4291 0.1>
          {arriveTangent=-0.7251, arriveTangentWeight=0.1918, leaveTangent=-0.7251, leaveTangentWeight=0.058, time=0.5086, value=1}
          <at 0.567 0>
          <at 0.6229 1>
          <at 0.6808 0>
          <at 0.7553 1>
          <at 0.8163 0>
          {arriveTangentWeight=0.0138, leaveTangentWeight=0.0174, time=0.8577, value=1}
          <at 0.9072 0>
          {arriveTangentWeight=0.0841, leaveTangentWeight=0.151, time=0.9394, value=1}
          {arriveTangentWeight=0.198, leaveTangentWeight=0.1375, time=0.9536, value=1}
          <at 1 0>
         ]}}}
      Scale Curve=0.3
     )
    )
   ) {isSetParameters=true}
   &ScaleMeshSize(Scale Factor=&MakeVector(
     X=&FloatFromCurve(FloatCurve=DataInterface() {propertyValues {Curve {keys [{tangentMode=RCTM_User, time=0, value=1}, {tangentMode=RCTM_User, time=0.5, value=0.5}]}}})
     Y=&FloatFromCurve(FloatCurve=DataInterface() {propertyValues {Curve {keys [{tangentMode=RCTM_User, time=0, value=1}, {tangentMode=RCTM_User, time=0.5, value=0.5}]}}})
     Z=&FN_FrontEndFloatModifier(
      Gameplay=&FloatFromCurve(FloatCurve=DataInterface() {propertyValues {Curve {keys [<at 0 0.6>,<at 0.9 1>,<at 0.9993 0>]}}})
      Front End=0
     )
    ))
   /SetVariables(
    Particles.Flame2Scale:Vector3f=&Multiply_Vector(
     A=$Particles.Initial.Flame2Scale
     B=&MakeVector(
      X=&FloatFromCurve(FloatCurve=DataInterface() {propertyValues {Curve {keys [<at 0 1>,<at 0.9104 1>,<at 1 1>]}}})
      Y=1
      Z=&FloatFromCurve(FloatCurve=DataInterface() {propertyValues {Curve {keys [<at 0 0.6>,<at 0.97 1>,<at 1 0>]}}})
     )
    )
    Particles.FlameCoreScale:Vector3f=&Multiply_Vector(
     A=$Particles.Initial.FlameCoreScale
     B=&MakeVector(
      X=&FloatFromCurve(FloatCurve=DataInterface() {propertyValues {Curve {keys [<at 0 1>,<at 0.95 1>,<at 1 1>]}}})
      Y=1
      Z=&FloatFromCurve(FloatCurve=DataInterface() {propertyValues {Curve {keys [<at 0 0.6>,<at 0.97 1>,<at 1 0>]}}})
     )
    )
   ) {isSetParameters=true}
   /SetVariables(
    Particles.Flame2Scale:Vector3f=&Multiply_Vector(
     A=$Particles.Flame2Scale
     B=&MakeVector(
      X=&FloatFromCurve(FloatCurve=DataInterface() {propertyValues {Curve {keys [{tangentMode=RCTM_User, time=0, value=0}, {tangentMode=RCTM_User, time=0.5, value=1}]}}})
      Y=&FloatFromCurve(FloatCurve=DataInterface() {propertyValues {Curve {keys [{tangentMode=RCTM_User, time=0, value=0}, {tangentMode=RCTM_User, time=0.5, value=1}]}}})
      Z=1
     )
    )
    Particles.FlameCoreScale:Vector3f=&Multiply_Vector(
     A=$Particles.FlameCoreScale
     B=&Multiply_Vector(
      A=&MakeVector(
       X=&FloatFromCurve(FloatCurve=DataInterface() {propertyValues {Curve {keys [{tangentMode=RCTM_User, time=0, value=1}, {tangentMode=RCTM_User, time=0.5, value=1}, <at 0.9093 1>]}}})
       Y=1
       Z=&FloatFromCurve(FloatCurve=DataInterface() {propertyValues {Curve {keys [<at 0 0.6>,<at 0.97 1>,<at 1 0>]}}})
      )
      B=&MakeVector(
       X=1
       Y=1
       Z=&FloatFromCurve(FloatCurve=DataInterface() {propertyValues {Curve {keys [<at 0 1>,<at 0.97 1>,<at 1 0>]}}})
      )
     )
    )
   ) {isSetParameters=true}
   &DynamicMaterialParameters(
    Index 0 Param 1=&Lerp_Float(A=0.3, B=0.6, Alpha=1)
    Module.Index0Float0Write=true
    Module.Index0Float1Write=true
    Module.Index0Float2Write=true
    Module.Index0Float3Write=true
   )
   /SetVariables(
    Particles.Wind Color:LinearColor=/Niagara/DynamicInputs/TypeConversions/MakeLinearColorFromFloat(Value:NiagaraFloat=&FloatFromCurve(FloatCurve=DataInterface() {propertyValues {Curve {
         defaultValue=1
         keys [
          {arriveTangent=9.0725, interpMode=RCIM_Cubic, leaveTangent=9.0725, tangentMode=RCTM_User, time=0, value=0}
          {arriveTangent=-2.5249, interpMode=RCIM_Cubic, leaveTangent=-2.5249, tangentMode=RCTM_User, time=0.1408, value=0.9847}
          {arriveTangent=-0.0405, interpMode=RCIM_Cubic, leaveTangent=-0.0405, tangentMode=RCTM_User, time=1, value=0}
         ]
        }}}))
    Particles.Wind Scale:Vector3f=<1 0.5 0.5>
   ) {isSetParameters=true}
   /SetVariables(
    Particles.Flare Color:LinearColor=&MakeLinearColorFromVectorAndFloat(
     "Vector (RGB)"=&MakeVectorFromLinearColorRGB(Color=$Particles.Initial.Flare Color)
     "Float (Alpha)"=&FloatFromCurve(
      FloatCurve=DataInterface() {propertyValues {Curve {keys [
          {arriveTangent=-0.9, time=0, value=0}
          {arriveTangent=-0.0998, time=0.05, value=1}
          {arriveTangent=-1.0872, leaveTangent=-1.0872, time=0.0688, value=0}
          <at 0.1141 0.95>
          <at 0.1507 0>
          <at 0.1987 1>
          <at 0.2343 0>
          <at 0.277 1>
          <at 0.316 0>
          <at 0.3728 1>
          <at 0.4234 0>
          <at 0.5255 0.95>
          <at 0.5315 0>
          <at 0.6404 1>
          <at 0.6954 0>
          {arriveTangent=-1.5912, leaveTangent=-1.5912, time=0.7682, value=1}
          <at 0.8633 0>
          <at 0.9501 1>
          <at 1.0002 0>
         ]}}}
      Scale Curve=3
     )
    )
    Particles.Flare Sprite Size:Vector2f=&Multiply_Vector2D(
     A=$Particles.Initial.Flare Sprite Size
     B=&Vector2DFromFloat(Value=&FloatFromCurve(FloatCurve=DataInterface() {propertyValues {Curve {keys [
           {interpMode=RCIM_Cubic, leaveTangentWeight=0.2638, tangentMode=RCTM_User, time=0, value=1}
           {arriveTangentWeight=0.3731, interpMode=RCIM_Cubic, leaveTangentWeight=0.0983, tangentMode=RCTM_User, time=0.8982, value=1}
           <at 1 0 RCIM_Cubic>
          ]}}}))
    )
   ) {isSetParameters=true}
   &FN_Light_Attributes(
    Color=&ColorFromCurve(ColorCurve=DataInterface() {propertyValues {
       AlphaCurve {keys [<at 0 1>,<at 0.8973 1>,<at 1 0>]}
       BlueCurve {keys [<at 0 2.501 RCIM_Cubic>]}
       GreenCurve {keys [<at 0 6.6555 RCIM_Cubic>]}
       RedCurve {keys [<at 0 12 RCIM_Cubic>]}
      }})
    Create Color Attribute=true
    Create Exponent Attribute=true
    Create Position Attribute=true
    Create Radius Attribute=true
    Enable Particle Light Enable=true
    Exponent=2
    Light Enabled=true
    Position=<60 30 10>
    Radius=100
    Scale Color=&FloatFromCurve(
     FloatCurve=DataInterface() {propertyValues {Curve {keys [
         {arriveTangent=4.4475, time=0, value=0}
         {arriveTangent=5.3518, leaveTangent=-3.1774, time=0.1, value=1.2}
         {arriveTangent=-1.4972, leaveTangent=-1.4972, time=0.1523, value=0.7}
         {arriveTangent=-0.1667, leaveTangent=-0.1667, time=0.2052, value=1.1711}
         <at 0.258 0.7713>
         <at 0.2926 1.1775>
         {arriveTangent=-0.25, leaveTangent=-0.25, time=0.3608, value=0.689}
         <at 0.3865 1.193>
         <at 0.4263 0.8643>
         <at 0.4931 1.1191>
         {arriveTangent=-0.25, leaveTangent=-0.25, time=0.5428, value=0.7548}
         <at 0.5846 1.1063>
         <at 0.6571 0.8076>
         <at 0.6976 1.1685>
         {arriveTangent=0.1667, leaveTangent=0.1667, time=0.7542, value=0.7713}
         <at 0.7906 1.1138>
         <at 0.8535 0.8807>
         {arriveTangent=-3.3333, leaveTangent=-3.3333, time=0.9, value=1.2}
         <at 1 0>
        ]}}}
     Scale Curve=&Multiply_Float(A=3, B=&FNTimeOfDayScalar(Night Scale=0.5, Morning Scale=0.8, Evening Scale=0.8))
    )
   )
  ]
  rendererClasses [
   &NiagaraMeshRendererProperties
   &NiagaraSpriteRendererProperties
   &NiagaraLightRendererProperties
  ]
  renderers [
   {rendererClass=&NiagaraMeshRendererProperties, rendererIndex=0}
   {rendererClass=&NiagaraMeshRendererProperties, rendererIndex=1}
   {rendererClass=&NiagaraMeshRendererProperties, rendererIndex=2}
   {rendererClass=&NiagaraMeshRendererProperties, rendererIndex=3}
   {rendererClass=&NiagaraMeshRendererProperties, rendererIndex=4}
   {rendererClass=&NiagaraSpriteRendererProperties, rendererIndex=5}
   {rendererClass=&NiagaraLightRendererProperties, rendererIndex=6}
  ]
 }
 Missile_Smoke {
  emitterUpdateScript [
   &EmitterState(Life Cycle Mode=::NewEnumerator1, Loop Behavior=::NewEnumerator1, Loop Duration=2.9, UseLoopDelay=false)
   &SpawnRate(SpawnGroup=1, SpawnRate=40, Use Spawn Probability=false) {module=SpawnRate001}
   &SpawnRate(SpawnRate=40, Use Spawn Probability=false)
  ]
  particleSpawnScript [
   &InitializeParticle(
    Color Mode=::NewEnumerator0
    Lifetime=&RandomRangeFloat(Minimum=0.15, Maximum=0.2)
    Sprite Rotation Mode=::NewEnumerator2
    UsePositionOffset=false
    Write Material Random=true
   )
   /SetVariables(
    Particles.Smoke Color:LinearColor=&RandomRangeLinearColor(Minimum=<rgba 0.6771 0.6771 0.6771 0.15>, Maximum=<rgba 0.6875 0.6648 0.6624 0.3>)
    Particles.Smoke Sprite Size:Vector2f=&Vector2DFromFloat(Value=&RandomRangeFloat(Minimum=40, Maximum=70))
   ) {isSetParameters=true}
   /SetVariables(
    Particles.Smoke Streak Color:LinearColor=&RandomRangeLinearColor(Minimum=<rgba 0.2273 0.2356 0.2604 0.5>, Maximum=<rgba 0.0848 0.0857 0.1094 0.7>)
    Particles.Smoke Streak Sprite Size:Vector2f=&UniformRangedVector2D(Minimum=<15 75>, Maximum=<30 90>)
   ) {isSetParameters=true}
   /SetVariables(
    Particles.Smoke Streak Color 2:LinearColor=&RandomRangeLinearColor(Minimum=<rgba 0.2273 0.2356 0.2604 0.5>, Maximum=<rgba 0.0848 0.0857 0.1094 0.7>)
    Particles.Smoke Streak Sprite Size 2:Vector2f=&UniformRangedVector2D(Minimum=<15 55>, Maximum=<20 60>)
   ) {isSetParameters=true}
   &SkeletalMeshLocation(
    Apply Sampled Orientation to Mesh Particles=true
    Bone Sampling Mode=::NewEnumerator4
    Inherit Velocity=false
    MaskSpawnGroup=false
    Sampled Position Offset=&MergeHelper_VectorSwitch(0=<0 0 -35>, 1=<0 0 -40>, 2=?, 3=?, Index=0, Num=::NewEnumerator0)
    Sampled Position Offset Coordinate Space:ENiagaraMeshTransforms=::NewEnumerator3
    Skeletal Mesh=$System.Skeletal Mesh
   )
   &AddVelocity(
    Cone Angle=&MergeHelper_FloatSwitch(0=10, 1=70, 2=?, 3=?, Index=&SpawnGroup(), Num=::NewEnumerator0)
    Cone Axis=<0.5 -0.5 1>
    Inner Cone Angle=15
    Module.UseConeFalloff=true
    Module.UseVelDistribution=false
    Override Seed=false
    Speed Falloff From Cone Axis=1.2784
    Velocity Mode=::NewEnumerator2
    Velocity Speed=&MergeHelper_FloatSwitch(
     0=&RandomRangeFloat(Minimum=150, Maximum=200)
     1=&RandomRangeFloat(Minimum=150, Maximum=200)
     2=?
     3=?
     Index=&SpawnGroup()
     Num=::NewEnumerator0
    )
   )
   &CameraOffset(Camera Offset Amount=-10)
   &ShapeLocation(MaskSpawnGroup=true, Override Seed=false, Sphere Radius=10, UseOwnerScale=true)
   /SetVariables(Particles.VisibilityTag=&SpawnGroup()) {isSetParameters=true}
  ]
  particleUpdateScript [
   &ParticleState()
   /SetVariables(
    Particles.Smoke Color:LinearColor=&MakeLinearColorFromVectorAndFloat(
     "Vector (RGB)"=&MakeVectorFromLinearColorRGB(Color=$Particles.Initial.Smoke Color)
     "Float (Alpha)"=&FloatFromCurve(FloatCurve=DataInterface() {propertyValues {Curve {keys [<at 0 0 RCIM_Cubic>,<at 0.5 1 RCIM_Cubic>,<at 1 0 RCIM_Cubic>]}}}, Scale Curve=0.2)
    )
    Particles.Smoke Sprite Size:Vector2f=&Multiply_Vector2D(
     A=$Particles.Initial.Smoke Sprite Size
     B=&Vector2DFromFloat(Value=&FloatFromCurve(FloatCurve=DataInterface() {propertyValues {Curve {keys [<at 0 0.3>,<at 1 1>]}}}))
    )
   ) {isSetParameters=true}
   /SetVariables(
    Particles.Smoke Streak Color:LinearColor=&MultiplyLinearColor(
     Linear Color=$Particles.Initial.Smoke Streak Color
     Scale Factor=&MakeVector4_FromLinearColor(LinearColor=&ColorFromCurve(ColorCurve=DataInterface() {propertyValues {
         AlphaCurve {keys [
           <at 0.0948 0>
           {arriveTangent=0.5207, interpMode=RCIM_Cubic, leaveTangent=0.5207, time=0.3829, value=0.25}
           <at 0.5749 0.25>
           <at 1 0 RCIM_Cubic>
          ]}
         BlueCurve {keys [<at 0 4.7917 RCIM_Cubic>,<at 1 1 RCIM_Cubic>]}
         GreenCurve {keys [<at 0 4.8426 RCIM_Cubic>,<at 1 1 RCIM_Cubic>]}
         RedCurve {keys [<at 0 5 RCIM_Cubic>,<at 1 1 RCIM_Cubic>]}
        }}))
    )
    Particles.Smoke Streak DynamicMaterialParameter1:Vector4f=&MakeVector4(
     X=&Multiply_Float(A=10, B=1)
     Y=0
     Z=0
     W=&FloatFromCurve(FloatCurve=DataInterface() {propertyValues {Curve {keys [<at 0 0>,<at 1 0.5>]}}})
    )
    Particles.Smoke Streak DynamicMaterialParameter2:Vector4f=&MakeVector4(
     X=&FloatFromCurve(
      FloatCurve=DataInterface() {propertyValues {Curve {keys [
          {interpMode=RCIM_Cubic, tangentMode=RCTM_User, time=0, value=0}
          {arriveTangent=0.6295, interpMode=RCIM_Cubic, leaveTangent=0.6295, tangentMode=RCTM_User, time=1, value=1}
         ]}}}
      Scale Curve=0.2
     )
     Y=&FloatFromCurve(FloatCurve=DataInterface() {propertyValues {Curve {keys [<at 0 1>,<at 1 0>]}}})
     Z=&FloatFromCurve(
      FloatCurve=DataInterface() {propertyValues {Curve {keys [{interpMode=RCIM_Cubic, tangentMode=RCTM_User, time=0.2, value=0.5}, {interpMode=RCIM_Cubic, tangentMode=RCTM_User, time=1, value=1}]}}}
      Scale Curve=2
     )
     W=&FloatFromCurve(FloatCurve=DataInterface() {propertyValues {Curve {keys [<at 0 0.3>,<at 1 0>]}}}, Scale Curve=-0.3)
    )
    Particles.Smoke Streak Sprite Size:Vector2f=&Multiply_Vector2D(
     A=$Particles.Initial.Smoke Streak Sprite Size
     B=&Vector2DFromCurve(
      Vector2Curve=DataInterface() {propertyValues {
        XCurve {keys [{arriveTangent=2.1678, interpMode=RCIM_Cubic, leaveTangent=2.1678, tangentMode=RCTM_User, time=0, value=0.3}, <at 1 1 RCIM_Cubic>]}
        YCurve {keys [{arriveTangent=2.3744, interpMode=RCIM_Cubic, leaveTangent=2.3744, tangentMode=RCTM_User, time=0, value=0.3}, <at 1 2 RCIM_Cubic>]}
       }}
      ScaleCurve=&Vector2DFromFloat(Value=2)
     )
    )
   ) {isSetParameters=true}
   /SetVariables(
    Particles.Smoke Streak Color 2:LinearColor=&MultiplyLinearColor(
     Linear Color=$Particles.Initial.Smoke Streak Color 2
     Scale Factor=&MakeVector4_FromLinearColor(LinearColor=&ColorFromCurve(ColorCurve=DataInterface() {propertyValues {
         AlphaCurve {keys [
           <at 0.0665 0>
           {arriveTangent=0.3934, interpMode=RCIM_Cubic, leaveTangent=0.3934, time=0.4165, value=0.2}
           <at 0.5749 0.2>
           <at 1 0 RCIM_Cubic>
          ]}
         BlueCurve {keys [<at 0 4.7917 RCIM_Cubic>,<at 1 1 RCIM_Cubic>]}
         GreenCurve {keys [<at 0 4.8426 RCIM_Cubic>,<at 1 1 RCIM_Cubic>]}
         RedCurve {keys [<at 0 5 RCIM_Cubic>,<at 1 1 RCIM_Cubic>]}
        }}))
    )
    Particles.Smoke Streak Sprite Size 2:Vector2f=&Multiply_Vector2D(
     A=$Particles.Initial.Smoke Streak Sprite Size 2
     B=&Vector2DFromCurve(
      Vector2Curve=DataInterface() {propertyValues {
        XCurve {keys [{arriveTangent=2.1678, interpMode=RCIM_Cubic, leaveTangent=2.1678, tangentMode=RCTM_User, time=0, value=0.3}, <at 1 1 RCIM_Cubic>]}
        YCurve {keys [{arriveTangent=2.3744, interpMode=RCIM_Cubic, leaveTangent=2.3744, tangentMode=RCTM_User, time=0, value=0.3}, <at 1 1.3 RCIM_Cubic>]}
       }}
      ScaleCurve=&Vector2DFromFloat(Value=2)
     )
    )
   ) {isSetParameters=true}
   /SetVariables(Particles.Smoke SubImageIndex:NiagaraFloat=&FloatFromCurve(FloatCurve=DataInterface() {propertyValues {Curve {keys [<at 0 32>,<at 1 64>]}}})) {isSetParameters=true}
   &SolveForcesAndVelocity(Clamp Velocity=false, Limit Acceleration=false)
  ]
  rendererClasses [&NiagaraSpriteRendererProperties]
  renderers [
   {rendererClass=&NiagaraSpriteRendererProperties}
   {rendererClass=&NiagaraSpriteRendererProperties, rendererIndex=1}
  ]
 }
 Shrapnel {
  emitterUpdateScript [
   &EmitterState(Life Cycle Mode=::NewEnumerator1, Loop Behavior=::NewEnumerator1, Loop Delay=4.3, Loop Duration=0.9, UseLoopDelay=true)
   &SpawnRate(SpawnGroup=1, SpawnRate=150, Use Spawn Probability=false) {module=SpawnRate001}
   &SpawnRate(
    SpawnRate=&FloatFromCurve(FloatCurve=DataInterface() {propertyValues {Curve {keys [<at 0.3 0.3>,<at 1 1>]}}}, CurveIndex=$Emitter.NormalizedLoopAge, Scale Curve=150)
    Use Spawn Probability=false
   )
  ]
  particleSpawnScript [
   &InitializeParticle(
    Color Mode=::NewEnumerator0
    Lifetime=&MergeHelper_FloatSwitch(
     0=&RandomRangeFloat(Minimum=0.18, Maximum=0.4)
     1=&RandomRangeFloat(Minimum=0.5, Maximum=0.85)
     2=?
     3=?
     Index=&SpawnGroup()
     Num=::NewEnumerator0
    )
    Sprite Rotation Mode=::NewEnumerator2
    Sprite Size Mode=::NewEnumerator1
    Uniform Sprite Size Max=50
    Uniform Sprite Size Min=25
    UsePositionOffset=false
    Write Material Random=true
   )
   &PartitionParticles(Number Of Partitions=7)
   &SkeletalMeshLocation(
    Bone / Socket Index=&SelectIntFromArray(Direct Array Index=$Particles.PartitionParticles.Partition, Int Selection Array=DataInterface() {propertyValues {IntData [3,5,6,8,4,0,7]}})
    MaskSpawnGroup=false
    Mesh Sampling Type=::NewEnumerator5
    Sampled Position Offset=<0 0 -3>
    Skeletal Mesh=$System.Skeletal Mesh
    Socket Sampling Mode=::NewEnumerator5
   )
   &AddVelocity(
    Cone Angle=35
    Cone Axis=<0 0 1>
    Module.UseConeFalloff=false
    Module.UseVelDistribution=false
    Override Seed=false
    Velocity Mode=::NewEnumerator2
    Velocity Speed=&MergeHelper_FloatSwitch(
     0=&FloatFromCurve(
      FloatCurve=DataInterface() {propertyValues {Curve {keys [{arriveTangent=-7.5919, interpMode=RCIM_Cubic, leaveTangent=-7.5919, tangentMode=RCTM_User, time=0, value=1}, <at 0.2 0.5 RCIM_Cubic>]}}}
      CurveIndex=$Emitter.LoopedAge
      Scale Curve=&RandomRangeFloat(Minimum=-1, Maximum=-2)
     )
     1=15
     2=?
     3=?
     Index=&SpawnGroup()
     Num=::NewEnumerator0
    )
   )
   &CameraOffset(Camera Offset Amount=15)
   /SetVariables(Particles.Smoke Sprite Size:Vector2f=&Vector2DFromFloat(Value=&RandomRangeFloat(Minimum=50, Maximum=80))) {isSetParameters=true}
   /SetVariables(Particles.VisibilityTag=&SpawnGroup()) {isSetParameters=true}
   &InheritVelocity(
    Inherited Velocity Amount Scale=<0 0 1>
    Inherited Velocity Speed Limit:NiagaraFloat=&MergeHelper_FloatSwitch(0=1000, 1=0, 2=?, 3=?, Index=&SpawnGroup(), Num=::NewEnumerator0)
    Speed Limit=true
    UseSpeedThreshold=false
   )
  ]
  particleUpdateScript [
   &ParticleState()
   &ScaleColor(
    Scale Alpha=&Multiply_Float(
     A=1
     B=&FloatFromCurve(FloatCurve=DataInterface() {propertyValues {Curve {keys [
          {arriveTangent=0.8687, interpMode=RCIM_Cubic, leaveTangent=0.8687, tangentMode=RCTM_User, time=0.1, value=0}
          {arriveTangent=0.0152, interpMode=RCIM_Cubic, leaveTangent=0.0152, tangentMode=RCTM_User, time=0.4, value=1}
          <at 0.8 0 RCIM_Cubic>
         ]}}})
    )
    Scale RGB=&MakeVectorFromLinearColorRGB(Color=&FNTimeOfDayColor(
      Night Color=<rgba 2.5 0.2984 0 1>
      Morning Color=<rgba 2.5 0.3637 0.0061 1>
      Day Color=<rgba 4 0.3734 0 1>
      Evening Color=<rgba 2.5 0.4793 0 1>
     ))
    ScaleA=true
    ScaleRGB=true
   ) {module=ScaleColor002}
   &ScaleSpriteSize(Uniform Curve Sprite Scale=DataInterface() {propertyValues {Curve {keys [
        {arriveTangent=0.5006, interpMode=RCIM_Cubic, leaveTangent=0.5006, leaveTangentWeight=0.7039, tangentMode=RCTM_User, time=0, value=0.725}
        {arriveTangent=-0.225, interpMode=RCIM_Cubic, leaveTangent=-0.225, time=0.6526, value=1.45}
        {arriveTangentWeight=0.7039, interpMode=RCIM_Cubic, tangentMode=RCTM_User, time=1, value=0.5}
       ]}}})
   /Niagara/Modules/Update/SubUV/SubUVAnimation(Number Of Frames=&FloatFromCurve(
     FloatCurve=DataInterface() {propertyValues {Curve {keys [{arriveTangent=4.0889, interpMode=RCIM_Cubic, leaveTangent=4.0889, tangentMode=RCTM_User, time=0, value=56}, <at 0.9 60>]}}}
     Scale Curve=&MaskFloatBySpawnGroup(Masked Float=0.7, Passthrough Float=1.25)
    ))
   &DynamicMaterialParameters(
    Index 0 Param 1=1
    Index 0 Param 2=&RandomRangeFloat(Minimum=0, Maximum=4)
    Index 0 Param 3=&FloatFromCurve(FloatCurve=DataInterface() {propertyValues {Curve {keys [<at 0 0.5 RCIM_Cubic>, {arriveTangent=1.5, interpMode=RCIM_Cubic, leaveTangent=1.5, time=0.5235, value=0.5}, <at 1 2>]}}})
    Index 0 Param 4=2.5
    Module.Index0Float0Write=true
    Module.Index0Float1Write=true
    Module.Index0Float2Write=true
    Module.Index0Float3Write=true
   )
   /Niagara/Modules/Update/Forces/V2/CurlNoiseForce(Bias Noise=true, Bias Noise Field:Vector3f=<0 0 1>, Noise Strength=500, Pan Noise=false)
   &Drag(
    Drag=&MergeHelper_FloatSwitch(
     0=0
     1=&FloatFromCurve(FloatCurve=DataInterface() {propertyValues {Curve {keys [<at 0 1>,<at 1 0>]}}}, Scale Curve=4)
     2=?
     3=?
     Index=&SpawnGroup()
     Num=::NewEnumerator0
    )
    Use Linear Drag=true
    Use Rotational Drag=true
   )
   &SolveForcesAndVelocity(Clamp Velocity=false, Limit Acceleration=false)
   /SetVariables(
    Particles.Smoke Color:LinearColor=&MultiplyLinearColor(
     Linear Color=&ColorFromCurve(ColorCurve=DataInterface() {propertyValues {
        AlphaCurve {keys [<at 0.006 0>,<at 0.163 0.45 RCIM_Cubic>,<at 1 0 RCIM_Cubic>]}
        BlueCurve {keys [<at 0 0.0468 RCIM_Cubic>]}
        GreenCurve {keys [<at 0 0.1733 RCIM_Cubic>]}
        RedCurve {keys [<at 0 0.9427 RCIM_Cubic>]}
       }})
     Scale Factor=&MakeVector4_FromLinearColor(LinearColor=&ColorFromCurve(
       ColorCurve=DataInterface() {propertyValues {
         AlphaCurve {keys [
           {arriveTangent=1.4278, interpMode=RCIM_Cubic, time=4.6993, value=0.75}
           {arriveTangent=-1.0843, interpMode=RCIM_Cubic, leaveTangent=-1.0843, time=5, value=0.5}
           <at 5.391 0 RCIM_Cubic>
          ]}
         BlueCurve {keys [
           {arriveTangentWeight=1.6042, interpMode=RCIM_Cubic, leaveTangentWeight=0.2176, tangentMode=RCTM_User, time=4.5, value=1}
           <at 4.7 0.05 RCIM_Cubic>
          ]}
         GreenCurve {keys [
           {arriveTangentWeight=1.6042, interpMode=RCIM_Cubic, leaveTangentWeight=0.2176, tangentMode=RCTM_User, time=4.5, value=1}
           <at 4.7 0.05 RCIM_Cubic>
          ]}
         RedCurve {keys [
           {arriveTangentWeight=1.6042, interpMode=RCIM_Cubic, leaveTangentWeight=0.2176, tangentMode=RCTM_User, time=4.5, value=1}
           <at 4.7 0.05 RCIM_Cubic>
          ]}
        }}
       CurveIndex=$System.Age
      ))
    )
    Particles.Smoke Sprite Size:Vector2f=&Multiply_Vector2D(
     A=$Particles.Initial.Smoke Sprite Size
     B=&Vector2DFromFloat(Value=&FloatFromCurve(FloatCurve=DataInterface() {propertyValues {Curve {keys [
           {arriveTangent=1.5344, interpMode=RCIM_Cubic, leaveTangent=1.5344, leaveTangentWeight=1.9568, tangentMode=RCTM_User, time=0, value=0.5}
           <at 1 1.2 RCIM_Cubic>
          ]}}}))
    )
    Particles.Smoke Streak DynamicMaterialParameter1:Vector4f=&MakeVector4(X=1, Y=1, Z=1, W=1)
   ) {isSetParameters=true}
   /SetVariables(Particles.Smoke SubImageIndex:NiagaraFloat=&FloatFromCurve(FloatCurve=DataInterface() {propertyValues {Curve {keys [<at 0 0>,<at 1 64>]}}})) {isSetParameters=true}
  ]
  rendererClasses [&NiagaraSpriteRendererProperties]
  renderers [
   {rendererClass=&NiagaraSpriteRendererProperties}
   {rendererClass=&NiagaraSpriteRendererProperties, rendererIndex=1}
  ]
 }
}
summary {
 systemName=NS_Emote_AuroraOwl_Missile
 emitters [
  {
   emitterName=Missile_FX
   rendererClasses [
    &NiagaraMeshRendererProperties
    &NiagaraSpriteRendererProperties
    &NiagaraLightRendererProperties
   ]
  }
  {emitterName=Explosion_Burst, rendererClasses [&NiagaraSpriteRendererProperties]}
  {
   emitterName=ImpactFlare
   rendererClasses [
    &NiagaraSpriteRendererProperties
    &NiagaraLightRendererProperties
    &NiagaraMeshRendererProperties
   ]
  }
  {
   emitterName=Explosion_Flare
   rendererClasses [
    &NiagaraSpriteRendererProperties
    &NiagaraMeshRendererProperties
    &NiagaraLightRendererProperties
   ]
  }
  {emitterName=Shrapnel, rendererClasses [&NiagaraSpriteRendererProperties]}
  {emitterName=Missile_Smoke, rendererClasses [&NiagaraSpriteRendererProperties]}
  {emitterName=DirectionalSparks_GPU, simTarget=GPUComputeSim, rendererClasses [&NiagaraSpriteRendererProperties]}
 ]
 userVariables [
  User.bIsFrontEnd {type=NiagaraBool}
  User.bIsFrontEndPreview {type=NiagaraBool}
  User.EngineDelay {type=NiagaraFloat, defaultValue=0.16, description=""}
  User.ExplosionDelay {type=NiagaraFloat, defaultValue=0, description=""}
  User.FeetTrailScale {type=NiagaraFloat, defaultValue=0, description=""}
  User.ForwardDot {type=NiagaraFloat, defaultValue=0, description=""}
  User.HandTrailScale {type=NiagaraFloat, defaultValue=1, description=""}
  User.HitGround {type=NiagaraBool, defaultValue=true}
  User.RibbonIDOffset {type=NiagaraInt32}
 ]
}
